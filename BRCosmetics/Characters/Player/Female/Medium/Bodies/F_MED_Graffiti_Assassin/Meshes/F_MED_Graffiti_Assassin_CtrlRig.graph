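# Control Rig for the F_MED_Graffiti_Assassin body that procedurally offsets skirt, belt strap and belt master bones as the legs move.
# Thigh and right thigh-twist rotations relative to the pelvis become Euler angles, whose clamped remaps weight a chain of ModifyTransforms nodes.
# CrouchSprintingAlpha weights an extra front-left skirt shift.
# One belt step has zero weight, and several leftover nodes are disconnected from execution.
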ControlRigBlueprint /BRCosmetics/Characters/Player/Female/Medium/Bodies/F_MED_Graffiti_Assassin/Meshes/F_MED_Graffiti_Assassin_CtrlRig
graphs [
 RigVMFunctionLibrary {links [], nodes []}
 RigVMModel {nodes [
   RigUnit_BeginExecution = &RigUnit_BeginExecution()
   n1 = &RigVMFunction_MathQuaternionToEuler(
    RotationOrder=ZYX
    Value=&RigUnit_GetRelativeTransformForItem(Child=(Type=Bone,Name="thigh_l"), Parent=(Type=Bone,Name="pelvis"), bChildInitial=false, bParentInitial=false).RelativeTransform.Rotation
   )
   &RigUnit_ModifyTransforms(
    ItemToModify=((Item=(Type=Bone,Name="dyn_skirt_fr_01_l_MSTR"),Transform=(Rotation=(X=0,Y=0,Z=0,W=1),Translation=(X=-8,Y=-6,Z=5),Scale3D=(X=1,Y=1,Z=1))))
    Weight=&RigVMFunction_MathFloatRemap(
     SourceMaximum=0
     SourceMinimum=-130
     TargetMaximum=0
     TargetMinimum=1
     bClamp=true
     Value=&RigVMFunction_MathFloatAdd(B=0, A=@n1.Result.Z) {template="Add::Execute(in A,in B,out Result)"}.Result
    ).Result
    ItemToModify.0.Transform.Rotation=&RigVMFunction_MathQuaternionFromEuler(Euler=(X=40,Y=-100,Z=-20)).Result
    ExecutePin=@RigUnit_BeginExecution.ExecutePin
   )
   n2 = &RigVMFunction_MathQuaternionToEuler(
    RotationOrder=ZYX
    Value=&RigUnit_GetRelativeTransformForItem(Child=(Type=Bone,Name="thigh_l"), Parent=(Type=Bone,Name="pelvis"), bChildInitial=false, bParentInitial=false).RelativeTransform.Rotation
   )
   n3 = &RigVMFunction_MathQuaternionToEuler(
    RotationOrder=ZYX
    Value=&RigUnit_GetTransform(Item=(Type=Bone,Name="dyn_skirt_fr_mid_l_MSTR"), Space=LocalSpace, bInitial=true).Transform.Rotation
   )
   > &RigUnit_ModifyTransforms(
    ItemToModify=((Item=(Type=Bone,Name="dyn_skirt_fr_mid_l_MSTR"),Transform=(Rotation=(X=0,Y=0,Z=0,W=1),Translation=(X=-6,Y=-5,Z=0),Scale3D=(X=1,Y=1,Z=1))))
    Weight=&RigVMFunction_MathFloatRemap(
     SourceMaximum=-10
     SourceMinimum=-130
     TargetMaximum=0
     TargetMinimum=1
     bClamp=true
     Value=&RigVMFunction_MathFloatAdd(B=0, A=@n2.Result.Z) {template="Add::Execute(in A,in B,out Result)"}.Result
    ).Result
    ItemToModify.0.Transform.Rotation=&RigVMFunction_MathQuaternionFromEuler(Euler=/Script/RigVM.RigVMFunction_MathVectorAdd(B=(X=0,Y=-90,Z=30), A=@n3.Result) {is=struct RigVMUnitNode, template="Add::Execute(in A,in B,out Result)"}.Result).Result
   )
   n4 = &RigVMFunction_MathQuaternionToEuler(
    RotationOrder=ZYX
    Value=&RigUnit_GetRelativeTransformForItem(Child=(Type=Bone,Name="thigh_l"), Parent=(Type=Bone,Name="pelvis"), bChildInitial=false, bParentInitial=false).RelativeTransform.Rotation
   )
   n5 = &RigVMFunction_MathQuaternionToEuler(
    RotationOrder=ZYX
    Value=&RigUnit_GetTransform(Item=(Type=Bone,Name="dyn_skirt_01_l_MSTR"), Space=LocalSpace, bInitial=true).Transform.Rotation
   )
   > &RigUnit_ModifyTransforms(
    ItemToModify=((Item=(Type=Bone,Name="dyn_skirt_01_l_MSTR"),Transform=(Rotation=(X=0,Y=0,Z=0,W=1),Translation=(X=0,Y=0,Z=0),Scale3D=(X=1,Y=1,Z=1))))
    Weight=&RigVMFunction_MathFloatRemap(
     SourceMaximum=-65
     SourceMinimum=0
     TargetMinimum=0
     bClamp=true
     Value=&RigVMFunction_MathFloatAdd(B=0, A=@n4.Result.Y) {template="Add::Execute(in A,in B,out Result)"}.Result
    ).Result
    ItemToModify.0.Transform.Rotation=&RigVMFunction_MathQuaternionFromEuler(Euler=&RigVMFunction_MathVectorAdd(B=(X=0,Y=-100,Z=30), A=@n5.Result) {template="Add::Execute(in A,in B,out Result)"}.Result).Result
   )
   > &RigUnit_ModifyTransforms(
    ItemToModify=((Item=(Type=Bone,Name="dyn_skirt_fr_01_l_MSTR"),Transform=(Rotation=(X=0,Y=0,Z=0,W=1),Translation=(X=-3,Y=0,Z=3),Scale3D=(X=1,Y=1,Z=1))))
    Weight=RigVMVariableNode() {variable=CrouchSprintingAlpha}.Value
   )
   n6 = /Script/RigVM.RigVMFunction_MathFloatAbs(Value=&RigVMFunction_MathQuaternionToEuler(
     RotationOrder=ZYX
     Value=&RigUnit_GetRelativeTransformForItem(Child=(Type=Bone,Name="thigh_twist_01_r"), Parent=(Type=Bone,Name="pelvis"), bChildInitial=false, bParentInitial=false).RelativeTransform.Rotation
    ).Result.Z) {is=struct RigVMUnitNode, template="Absolute::Execute(in Value,out Result)"}
   > &RigUnit_ModifyTransforms(
    ItemToModify=((Item=(Type=Bone,Name="dyn_beltstrap_MSTR"),Transform=(Rotation=(X=0,Y=0,Z=0,W=1),Translation=(X=-4,Y=0,Z=4),Scale3D=(X=1,Y=1,Z=1))))
    Weight=&RigVMFunction_MathFloatRemap(
     SourceMaximum=70
     SourceMinimum=170
     TargetMinimum=0
     bClamp=true
     Value=&RigVMFunction_MathFloatAdd(B=@n6.Result) {template="Add::Execute(in A,in B,out Result)"}.Result
    ).Result
   )
   n7 = &RigVMFunction_MathQuaternionToEuler(
    RotationOrder=ZYX
    Value=&RigUnit_GetRelativeTransformForItem(Child=(Type=Bone,Name="thigh_twist_01_r"), Parent=(Type=Bone,Name="pelvis"), bChildInitial=false, bParentInitial=false).RelativeTransform.Rotation
   )
   n8 = &RigVMFunction_MathFloatAbs(Value=@n7.Result.Z) {template="Absolute::Execute(in Value,out Result)"}
   > &RigUnit_ModifyTransforms(
    ItemToModify=((Item=(Type=Bone,Name="dyn_skirt_fr_01_r_MSTR"),Transform=(Rotation=(X=0,Y=0,Z=0,W=1),Translation=(X=-2,Y=1,Z=6),Scale3D=(X=1,Y=1,Z=1))))
    Weight=&RigVMFunction_MathFloatRemap(
     SourceMaximum=70
     SourceMinimum=170
     TargetMinimum=0
     bClamp=true
     Value=&RigVMFunction_MathFloatAdd(B=@n8.Result) {template="Add::Execute(in A,in B,out Result)"}.Result
    ).Result
    ItemToModify.0.Transform.Rotation=&RigVMFunction_MathQuaternionFromEuler(Euler=(X=0,Y=-120,Z=0)).Result
   )
   n9 = &RigVMFunction_MathFloatAbs(Value=&RigVMFunction_MathQuaternionToEuler(
     RotationOrder=ZYX
     Value=&RigUnit_GetRelativeTransformForItem(Child=(Type=Bone,Name="thigh_r"), Parent=(Type=Bone,Name="pelvis"), bChildInitial=false, bParentInitial=false).RelativeTransform.Rotation
    ).Result.Z) {template="Absolute::Execute(in Value,out Result)"}
   > &RigUnit_ModifyTransforms(
    ItemToModify=((Item=(Type=Bone,Name="dyn_skirt_fr_mid_r_MSTR"),Transform=(Rotation=(X=0,Y=0,Z=0,W=1),Translation=(X=-5,Y=4,Z=0),Scale3D=(X=1,Y=1,Z=1))))
    Weight=&RigVMFunction_MathFloatRemap(
     SourceMaximum=80
     SourceMinimum=180
     TargetMinimum=0
     bClamp=true
     Value=&RigVMFunction_MathFloatAdd(B=0, A=@n9.Result) {template="Add::Execute(in A,in B,out Result)"}.Result
    ).Result
    ItemToModify.0.Transform.Rotation=&RigVMFunction_MathQuaternionFromEuler(Euler=(X=0,Y=-80,Z=10)).Result
   )
   n10 = &RigVMFunction_MathFloatAbs(Value=&RigVMFunction_MathQuaternionToEuler(
     RotationOrder=ZYX
     Value=&RigUnit_GetRelativeTransformForItem(Child=(Type=Bone,Name="thigh_r"), Parent=(Type=Bone,Name="pelvis"), bChildInitial=false, bParentInitial=false).RelativeTransform.Rotation
    ).Result.Z) {template="Absolute::Execute(in Value,out Result)"}
   > &RigUnit_ModifyTransforms(
    ItemToModify=((Item=(Type=Bone,Name="dyn_skirt_01_r_MSTR"),Transform=(Rotation=(X=0,Y=0,Z=0,W=1),Translation=(X=2,Y=3,Z=0),Scale3D=(X=1,Y=1,Z=1))))
    Weight=&RigVMFunction_MathFloatRemap(
     SourceMaximum=80
     SourceMinimum=150
     TargetMinimum=0
     bClamp=true
     Value=&RigVMFunction_MathFloatAdd(B=0, A=@n10.Result) {template="Add::Execute(in A,in B,out Result)"}.Result
    ).Result
    ItemToModify.0.Transform.Rotation=&RigVMFunction_MathQuaternionFromEuler(Euler=(X=0,Y=-45,Z=-10)).Result
   )
   n11 = &RigVMFunction_MathQuaternionToEuler(
    RotationOrder=ZYX
    Value=&RigUnit_GetRelativeTransformForItem(Child=(Type=Bone,Name="thigh_r"), Parent=(Type=Bone,Name="pelvis"), bChildInitial=false, bParentInitial=false).RelativeTransform.Rotation
   )
   > &RigUnit_ModifyTransforms(
    ItemToModify=((Item=(Type=Bone,Name="dyn_skirt_01_r_MSTR"),Transform=(Rotation=(X=0,Y=0,Z=0,W=1),Translation=(X=-5,Y=4,Z=0),Scale3D=(X=1,Y=1,Z=1))))
    Weight=&RigVMFunction_MathFloatRemap(
     SourceMaximum=80
     SourceMinimum=0
     TargetMinimum=0
     bClamp=true
     Value=&RigVMFunction_MathFloatAdd(B=0, A=@n11.Result.Y) {template="Add::Execute(in A,in B,out Result)"}.Result
    ).Result
    ItemToModify.0.Transform.Rotation=&RigVMFunction_MathQuaternionFromEuler(Euler=(X=0,Y=-120,Z=20)).Result
   )
   > &RigUnit_ModifyTransforms(
    ItemToModify=((Item=(Type=Bone,Name="dyn_belt_01_r_MSTR"),Transform=(Rotation=(X=0,Y=0,Z=0,W=1),Translation=(X=0,Y=0,Z=0),Scale3D=(X=1,Y=1,Z=1))))
    Weight=0
    ItemToModify.0.Transform.Rotation=&RigVMFunction_MathQuaternionFromEuler(Euler=(X=0,Y=-120,Z=0)).Result
   )
   n12 = &RigVMFunction_MathFloatAbs(Value=&RigVMFunction_MathQuaternionToEuler(
     RotationOrder=ZYX
     Value=&RigUnit_GetRelativeTransformForItem(Child=(Type=Bone,Name="thigh_twist_01_r"), Parent=(Type=Bone,Name="pelvis"), bChildInitial=false, bParentInitial=false).RelativeTransform.Rotation
    ).Result.Z) {template="Absolute::Execute(in Value,out Result)"}
   > n13 = &RigUnit_ModifyTransforms(
    ItemToModify=((Item=(Type=Bone,Name="dyn_belt_01_r_MSTR"),Transform=(Rotation=(X=0,Y=0,Z=0,W=1),Translation=(X=0,Y=0,Z=0),Scale3D=(X=1,Y=1,Z=1))))
    Weight=&RigVMFunction_MathFloatRemap(
     SourceMaximum=90
     SourceMinimum=170
     TargetMinimum=0
     bClamp=true
     Value=&RigVMFunction_MathFloatAdd(B=@n12.Result) {template="Add::Execute(in A,in B,out Result)"}.Result
    ).Result
    ItemToModify.0.Transform.Rotation=&RigVMFunction_MathQuaternionFromEuler(Euler=(X=-20,Y=-70,Z=50)).Result
   )
   RigVMRerouteNode(Value=@n13.ExecutePin)
   RigVMDispatchNode(
    False=0
    Condition=/Script/RigVM.RigVMFunction_MathFloatGreater(A=@n7.Result.Z) {is=struct RigVMUnitNode, template="Greater::Execute(in A,in B,out Result)"}.Result
    True=@n7.Result.Z
   ) {template="DISPATCH_RigVMDispatch_If(in Condition,in True,in False,out Result)"}
   &RigVMFunction_MathFloatAbs() {template="Absolute::Execute(in Value,out Result)"}
   /Script/RigVM.RigVMFunction_MathFloatLess(A=@n7.Result.Z) {is=struct RigVMUnitNode, template="Less::Execute(in A,in B,out Result)"}
   &RigVMFunction_MathFloatMul(A=-5, B=-1) {template="Multiply::Execute(in A,in B,out Result)"}
   &RigVMFunction_MathFloatMul(A=-5, B=-1) {template="Multiply::Execute(in A,in B,out Result)"}
   &RigVMFunction_MathFloatMul(A=-5, B=-1) {template="Multiply::Execute(in A,in B,out Result)"}
   n14 = &RigUnit_GetRelativeTransformForItem(Child=(Type=Bone,Name="thigh_l"), Parent=(Type=Bone,Name="pelvis"), bChildInitial=false, bParentInitial=false)
   &RigVMFunction_MathFloatRemap(
    SourceMaximum=0
    SourceMinimum=-130
    TargetMaximum=0
    TargetMinimum=1
    bClamp=true
    Value=&RigVMFunction_MathFloatAdd(B=0, A=&RigVMFunction_MathQuaternionToEuler(RotationOrder=ZYX, Value=@n14.RelativeTransform.Rotation).Result.Z) {template="Add::Execute(in A,in B,out Result)"}.Result
   )
   n15 = &RigUnit_GetTransform(Item=(Type=Bone,Name="dyn_belt_01_r_MSTR"), Space=LocalSpace, bInitial=true)
   &RigVMFunction_MathQuaternionFromEuler(Euler=&RigVMFunction_MathVectorAdd(B=(X=-90,Y=-90,Z=-50), A=&RigVMFunction_MathQuaternionToEuler(RotationOrder=ZYX, Value=@n15.Transform.Rotation).Result) {template="Add::Execute(in A,in B,out Result)"}.Result)
   &RigVMFunction_MathQuaternionFromEuler(Euler=(X=0,Y=-90,Z=0))
   n16 = &RigUnit_GetTransform(Item=(Type=Bone,Name="dyn_skirt_fr_mid_l_MSTR"), Space=LocalSpace, bInitial=true)
   &RigVMFunction_MathQuaternionFromEuler(Euler=&RigVMFunction_MathVectorAdd(B=(X=0,Y=-80,Z=30), A=&RigVMFunction_MathQuaternionToEuler(RotationOrder=ZYX, Value=@n16.Transform.Rotation).Result) {template="Add::Execute(in A,in B,out Result)"}.Result)
   &RigVMFunction_MathVectorAdd(
    B=(X=0,Y=-100,Z=-10)
    A=&RigVMFunction_MathQuaternionToEuler(
     RotationOrder=ZYX
     Value=&RigUnit_GetTransform(Item=(Type=Bone,Name="dyn_skirt_fr_mid_l_MSTR"), Space=LocalSpace, bInitial=true).Transform.Rotation
    ).Result
   ) {template="Add::Execute(in A,in B,out Result)"}
   &RigVMFunction_MathVectorAdd(
    B=(X=180,Y=-150,Z=0)
    A=&RigVMFunction_MathQuaternionToEuler(
     RotationOrder=ZYX
     Value=&RigUnit_GetTransform(Item=(Type=Bone,Name="dyn_skirt_01_r_MSTR"), Space=LocalSpace, bInitial=true).Transform.Rotation
    ).Result
   ) {template="Add::Execute(in A,in B,out Result)"}
   &RigVMFunction_MathVectorAdd(
    B=(X=180,Y=-60,Z=-50)
    A=&RigVMFunction_MathQuaternionToEuler(
     RotationOrder=ZYX
     Value=&RigUnit_GetTransform(Item=(Type=Bone,Name="dyn_skirt_fr_mid_r_MSTR"), Space=LocalSpace, bInitial=true).Transform.Rotation
    ).Result
   ) {template="Add::Execute(in A,in B,out Result)"}
   &RigVMFunction_MathVectorAdd(
    B=(X=180,Y=-90,Z=0)
    A=&RigVMFunction_MathQuaternionToEuler(
     RotationOrder=ZYX
     Value=&RigUnit_GetTransform(Item=(Type=Bone,Name="dyn_skirt_fr_mid_r_MSTR"), Space=LocalSpace, bInitial=true).Transform.Rotation
    ).Result
   ) {template="Add::Execute(in A,in B,out Result)"}
   &RigVMFunction_MathVectorAdd(
    B=(X=180,Y=-90,Z=-25)
    A=&RigVMFunction_MathQuaternionToEuler(
     RotationOrder=ZYX
     Value=&RigUnit_GetTransform(Item=(Type=Bone,Name="dyn_skirt_fr_mid_r_MSTR"), Space=LocalSpace, bInitial=true).Transform.Rotation
    ).Result
   ) {template="Add::Execute(in A,in B,out Result)"}
   &RigVMFunction_MathVectorAdd(
    B=(X=180,Y=-60,Z=-50)
    A=&RigVMFunction_MathQuaternionToEuler(
     RotationOrder=ZYX
     Value=&RigUnit_GetTransform(Item=(Type=Bone,Name="dyn_skirt_fr_mid_r_MSTR"), Space=LocalSpace, bInitial=true).Transform.Rotation
    ).Result
   ) {template="Add::Execute(in A,in B,out Result)"}
   n17 = &RigVMFunction_MathQuaternionToEuler(
    RotationOrder=ZYX
    Value=&RigUnit_GetRelativeTransformForItem(Child=(Type=Bone,Name="thigh_r"), Parent=(Type=Bone,Name="pelvis"), bChildInitial=false, bParentInitial=false).RelativeTransform.Rotation
   )
   n18 = &RigVMFunction_MathQuaternionToEuler(
    RotationOrder=ZYX
    Value=&RigUnit_GetTransform(Item=(Type=Bone,Name="dyn_skirt_01_r_MSTR"), Space=LocalSpace, bInitial=true).Transform.Rotation
   )
   &RigUnit_ModifyTransforms(
    ExecutePin=()
    ItemToModify=((Item=(Type=Bone,Name="dyn_belt_01_r_MSTR"),Transform=(Rotation=(X=0,Y=0,Z=0,W=1),Translation=(X=0,Y=0,Z=0),Scale3D=(X=1,Y=1,Z=1))))
    Weight=&RigVMFunction_MathFloatRemap(
     SourceMaximum=65
     SourceMinimum=0
     TargetMinimum=0
     bClamp=true
     Value=&RigVMFunction_MathFloatAdd(B=0, A=@n17.Result.Y) {template="Add::Execute(in A,in B,out Result)"}.Result
    ).Result
    ItemToModify.0.Transform.Rotation=&RigVMFunction_MathQuaternionFromEuler(Euler=&RigVMFunction_MathVectorAdd(B=(X=180,Y=-90,Z=0), A=@n18.Result) {template="Add::Execute(in A,in B,out Result)"}.Result).Result
   )
  ]}
]
hierarchy [
 root {loc=<0 0 0>, rot=<0 0 0 1>}
 pelvis {parent=root, loc=<0 -0.5301 98.0571>, rot=<0.0127 0.707 -0.0127 -0.707>}
 spine_01 {parent=pelvis, loc=<6.1208 -0.9328 0>, rot=<0 0 0.034 -0.9994>}
 spine_02 {parent=spine_01, loc=<6.5996 -0.2095 0>, rot=<0 0 0.0238 -0.9997>}
 spine_03 {parent=spine_02, loc=<6.4914 -0.1237 0>, rot=<0 0 0.0149 -0.9999>}
 spine_04 {parent=spine_03, loc=<6.9598 0.0794 0>, rot=<0 0 -0.0132 -0.9999>}
 spine_05 {parent=spine_04, loc=<6.9987 0 0>, rot=<0 0 -0.0015 -1>}
 clavicle_l {parent=spine_05, loc=<5.4831 -0.1053 -3.2946>, rot=<0.0465 -0.7289 -0.009 -0.6829>}
 upperarm_l {parent=clavicle_l, loc=<9.8658 -1.2136 -0.4105>, rot=<0.0164 -0.3245 -0.0269 -0.9454>}
 lowerarm_l {parent=upperarm_l, loc=<26.465 -0.8345 0.259>, rot=<0.0036 0.0161 -0.2298 -0.9731>}
 hand_l {parent=lowerarm_l, loc=<22.7355 -0.3203 -0.6105>, rot=<-0.6784 -0.0223 0.0735 -0.7307>}
 index_metacarpal_l {parent=hand_l, loc=<1.6535 0.0183 -2.96>, rot=<0.0158 -0.0124 -0.0173 -0.9996>}
 index_01_l {parent=index_metacarpal_l, loc=<8.3008 0.0357 -0.0524>, rot=<-0.0044 0.0485 0.0924 -0.9945>}
 index_02_l {parent=index_01_l, loc=<3.3587 -0.3004 0.0452>, rot=<-0.002 0.0027 0.3225 -0.9466>}
 index_03_l {parent=index_02_l, loc=<2.5578 -0.123 -0.0252>, rot=<-0.002 0.0067 0.2558 -0.9667>}
 middle_metacarpal_l {parent=hand_l, loc=<2.1856 -0.1725 -0.5984>, rot=<0.006 0.0168 -0.0329 -0.9993>}
 middle_01_l {parent=middle_metacarpal_l, loc=<7.3275 -0.2447 0.0129>, rot=<-0.0208 0.0042 0.0858 -0.9961>}
 middle_02_l {parent=middle_01_l, loc=<4.0123 0.1401 0.1175>, rot=<-0.0218 0.0258 0.3512 -0.9357>}
 middle_03_l {parent=middle_02_l, loc=<2.6077 0.0534 0.1373>, rot=<0.0082 0.0188 0.1708 -0.9851>}
 pinky_metacarpal_l {parent=hand_l, loc=<1.7616 -0.9753 3.3814>, rot=<-0.0397 0.0814 -0.069 -0.9935>}
 pinky_01_l {parent=pinky_metacarpal_l, loc=<7.1709 -0.4831 -0.1094>, rot=<-0.0654 -0.0035 0.2242 -0.9723>}
 pinky_02_l {parent=pinky_01_l, loc=<2.8274 0.0893 -0.1043>, rot=<0.0037 0.0011 0.2854 -0.9584>}
 pinky_03_l {parent=pinky_02_l, loc=<2.2046 0.0307 -0.0078>, rot=<-0.0002 0.0098 0.1445 -0.9895>}
 ring_metacarpal_l {parent=hand_l, loc=<2.2074 -0.3625 1.3638>, rot=<-0.0284 0.0722 -0.0458 -0.9959>}
 ring_01_l {parent=ring_metacarpal_l, loc=<7.0295 -0.6197 -0.1446>, rot=<-0.0489 -0.0252 0.1892 -0.9804>}
 ring_02_l {parent=ring_01_l, loc=<3.6174 0.4647 -0.1027>, rot=<0 0 0.2806 -0.9598>}
 ring_03_l {parent=ring_02_l, loc=<2.4152 0.114 0.0199>, rot=<0 0 0.1189 -0.9929>}
 thumb_01_l {parent=hand_l, loc=<1.3118 -0.3191 -2.459>, rot=<0.6546 -0.4042 0.0664 -0.6354>}
 thumb_02_l {parent=thumb_01_l, loc=<5.9336 -0.0448 0.0401>, rot=<-0.0037 0.0213 0.298 -0.9543>}
 thumb_03_l {parent=thumb_02_l, loc=<2.2223 0.3682 -0.0907>, rot=<0.0118 -0.013 0.0066 -0.9998>}
 lowerarm_twist_01_l {parent=lowerarm_l, loc=<18.316 0 0>, rot=<0 0 0 1>}
 lowerarm_twist_02_l {parent=lowerarm_l, loc=<10.6707 0 0>, rot=<0 0 0 1>}
 upperarm_twist_01_l {parent=upperarm_l, loc=<7.2135 -0.1134 0>, rot=<0 0 0 1>}
 upperarm_twist_02_l {parent=upperarm_l, loc=<19.7998 -0.1134 0>, rot=<0 0 0 1>}
 clavicle_r {parent=spine_05, loc=<5.4831 -0.1053 3.2946>, rot=<0.7289 0.0465 -0.6829 0.009>}
 upperarm_r {parent=clavicle_r, loc=<-9.8658 1.2136 0.4104>, rot=<0.0164 -0.3245 -0.0269 -0.9454>}
 lowerarm_r {parent=upperarm_r, loc=<-26.4651 0.8345 -0.259>, rot=<0.0036 0.0161 -0.2298 -0.9731>}
 hand_r {parent=lowerarm_r, loc=<-22.7356 0.3203 0.6106>, rot=<-0.6784 -0.0223 0.0735 -0.7307>}
 index_metacarpal_r {parent=hand_r, loc=<-1.6534 -0.0181 2.9601>, rot=<0.0158 -0.0124 -0.0173 -0.9996>}
 index_01_r {parent=index_metacarpal_r, loc=<-8.3008 -0.0357 0.0524>, rot=<-0.0044 0.0485 0.0924 -0.9945>}
 index_02_r {parent=index_01_r, loc=<-3.3589 0.3003 -0.0453>, rot=<-0.002 0.0027 0.3225 -0.9466>}
 index_03_r {parent=index_02_r, loc=<-2.5578 0.123 0.0252>, rot=<-0.002 0.0067 0.2558 -0.9667>}
 middle_metacarpal_r {parent=hand_r, loc=<-2.1856 0.1726 0.5984>, rot=<0.006 0.0168 -0.0329 -0.9993>}
 middle_01_r {parent=middle_metacarpal_r, loc=<-7.3273 0.2447 -0.0129>, rot=<-0.0208 0.0042 0.0858 -0.9961>}
 middle_02_r {parent=middle_01_r, loc=<-4.0124 -0.14 -0.1175>, rot=<-0.0218 0.0258 0.3512 -0.9357>}
 middle_03_r {parent=middle_02_r, loc=<-2.6077 -0.0533 -0.1373>, rot=<0.0082 0.0188 0.1708 -0.9851>}
 pinky_metacarpal_r {parent=hand_r, loc=<-1.7617 0.9753 -3.3814>, rot=<-0.0397 0.0814 -0.069 -0.9935>}
 pinky_01_r {parent=pinky_metacarpal_r, loc=<-7.1707 0.4832 0.1094>, rot=<-0.0654 -0.0035 0.2242 -0.9723>}
 pinky_02_r {parent=pinky_01_r, loc=<-2.8274 -0.0892 0.1043>, rot=<0.0037 0.0011 0.2854 -0.9584>}
 pinky_03_r {parent=pinky_02_r, loc=<-2.2045 -0.031 0.0078>, rot=<-0.0002 0.0098 0.1445 -0.9895>}
 ring_metacarpal_r {parent=hand_r, loc=<-2.2073 0.3627 -1.3638>, rot=<-0.0284 0.0722 -0.0458 -0.9959>}
 ring_01_r {parent=ring_metacarpal_r, loc=<-7.0297 0.6197 0.1446>, rot=<-0.0489 -0.0252 0.1892 -0.9804>}
 ring_02_r {parent=ring_01_r, loc=<-3.6171 -0.4648 0.1027>, rot=<0 0 0.2806 -0.9598>}
 ring_03_r {parent=ring_02_r, loc=<-2.4153 -0.1141 -0.0199>, rot=<0 0 0.1189 -0.9929>}
 thumb_01_r {parent=hand_r, loc=<-1.3119 0.3192 2.459>, rot=<0.6546 -0.4042 0.0664 -0.6354>}
 thumb_02_r {parent=thumb_01_r, loc=<-5.9335 0.0447 -0.0401>, rot=<-0.0037 0.0213 0.298 -0.9543>}
 thumb_03_r {parent=thumb_02_r, loc=<-2.2223 -0.3682 0.0907>, rot=<0.0118 -0.013 0.0066 -0.9998>}
 lowerarm_twist_01_r {parent=lowerarm_r, loc=<-18.3161 0 0>, rot=<0 0 0 1>}
 lowerarm_twist_02_r {parent=lowerarm_r, loc=<-10.6707 0 0>, rot=<0 0 0 1>}
 upperarm_twist_01_r {parent=upperarm_r, loc=<-7.2135 0.1134 0>, rot=<0 0 0 1>}
 upperarm_twist_02_r {parent=upperarm_r, loc=<-19.7998 0.1134 0>, rot=<0 0 0 1>}
 neck_01 {parent=spine_05, loc=<11.3138 0.1529 0>, rot=<0 0 -0.2003 -0.9797>}
 neck_02 {parent=neck_01, loc=<4.5627 0 0>, rot=<0 0 0.0504 -0.9987>}
 dyn_collar_01_l_MSTR {parent=spine_05, loc=<0.4299 9.6892 -5.8133>, rot=<0.4044 -0.6945 -0.2761 -0.5271>}
 dyn_collar_01_l {parent=dyn_collar_01_l_MSTR, loc=<0.2 0 0>, rot=<0 0 0 1>}
 dyn_collar_01_r_MSTR {parent=spine_05, loc=<0.4299 9.6892 5.8133>, rot=<0.5271 0.2761 0.6945 -0.4044>}
 dyn_collar_01_r {parent=dyn_collar_01_r_MSTR, loc=<0.2 -0.0002 -0.0002>, rot=<0 0 0 1>}
 thigh_l {parent=pelvis, loc=<-1.0855 -0.2109 -8.7053>, rot=<0.0009 0.0355 0.0155 -0.9993>}
 calf_l {parent=thigh_l, loc=<-42.5284 -0.0019 -0.7297>, rot=<0.0006 -0.0161 -0.0391 -0.9991>}
 calf_twist_01_l {parent=calf_l, loc=<-35.8483 -1.9197 -0.5794>, rot=<0.0237 0.0095 -0.0035 -0.9997>}
 calf_twist_02_l {parent=calf_l, loc=<-19.1546 -1.7943 -0.2638>, rot=<0.0237 0.0095 -0.0035 -0.9997>}
 foot_l {parent=calf_l, loc=<-46.7402 -2.5014 -0.592>, rot=<0.0427 -0.0238 0.0353 -0.9982>}
 ball_l {parent=foot_l, loc=<-3.5658 10.8066 -0.1544>, rot=<0.0001 -0.0001 -0.7063 -0.7079>}
 thigh_twist_01_l {parent=thigh_l, loc=<-27.952 0 0>, rot=<0.0013 0 0 -1>}
 thigh_r {parent=pelvis, loc=<-1.0853 -0.2109 8.7053>, rot=<0.0355 -0.0009 0.9993 0.0155>}
 calf_r {parent=thigh_r, loc=<42.5286 0.0019 0.7297>, rot=<0.0006 -0.0161 -0.0391 -0.9991>}
 calf_twist_01_r {parent=calf_r, loc=<35.8485 1.9197 0.5794>, rot=<0.0237 0.0095 -0.0035 -0.9997>}
 calf_twist_02_r {parent=calf_r, loc=<19.1546 1.7943 0.2638>, rot=<0.0237 0.0095 -0.0035 -0.9997>}
 foot_r {parent=calf_r, loc=<46.7403 2.5014 0.592>, rot=<0.0427 -0.0238 0.0353 -0.9982>}
 ball_r {parent=foot_r, loc=<3.5658 -10.8066 0.1544>, rot=<0.0001 -0.0001 -0.7063 -0.7079>}
 thigh_twist_01_r {parent=thigh_r, loc=<27.952 0 0>, rot=<0.0013 0 0 -1>}
 dyn_gun_MSTR {parent=thigh_twist_01_r, loc=<-7.6496 -0.2376 7.7404>, rot=<-0.0184 -0.0189 -0.0228 -0.9994>}
 dyn_gun {parent=dyn_gun_MSTR, loc=<0.2 0 0>, rot=<0 0 0 1>}
 dyn_knife_01_l_MSTR {parent=pelvis, loc=<7.9724 -9.3662 1.5696>, rot=<0.4938 -0.4537 0.6755 0.3066>}
 dyn_knife_01_l {parent=dyn_knife_01_l_MSTR, loc=<0.2 0 0>, rot=<0 0 0 1>}
 dyn_skirt_bk_01_l_MSTR {parent=pelvis, loc=<-0.0887 -10.558 -6.9177>, rot=<-0.0173 -0.6952 0.7184 -0.018>}
 dyn_skirt_bk_01_l {parent=dyn_skirt_bk_01_l_MSTR, loc=<0.2 0 0>, rot=<0 0 0 1>}
 dyn_skirt_bk_01_r_MSTR {parent=pelvis, loc=<-0.0888 -10.558 6.9177>, rot=<0.018 -0.7184 0.6952 0.0173>}
 dyn_skirt_bk_01_r {parent=dyn_skirt_bk_01_r_MSTR, loc=<0.2 0 0>, rot=<0 0 0 1>}
 dyn_pouch_l_B_MSTR {parent=pelvis, loc=<10.7623 -0.0044 -10.9929>, rot=<-0.0186 -0.9918 -0.0707 -0.1047>}
 dyn_pouch_l_B {parent=dyn_pouch_l_B_MSTR, loc=<0 0 0.6372>, rot=<0 0 0 1>}
 dyn_pouch_l_A_MSTR {parent=pelvis, loc=<10.6931 4.3688 -9.4851>, rot=<-0.0034 -0.938 -0.3291 -0.1085>}
 dyn_pouch_l_A {parent=dyn_pouch_l_A_MSTR, loc=<0 0 0.6372>, rot=<0 0 0 1>}
 dyn_skirt_01_l_MSTR {parent=pelvis, loc=<2.0852 -1.3537 -14.0176>, rot=<0.0057 -0.988 -0.0166 -0.1535>}
 dyn_skirt_01_l {parent=dyn_skirt_01_l_MSTR, loc=<0.2 0 0>, rot=<0 0 0 1>}
 dyn_skirt_02_l {parent=dyn_skirt_01_l, loc=<6.1888 0 0>, rot=<0 0 0 1>}
 dyn_skirt_fr_01_l_MSTR {parent=pelvis, loc=<4.9975 7.7332 -3.74>, rot=<-0.0496 -0.7623 -0.6379 -0.0976>}
 dyn_skirt_fr_01_l {parent=dyn_skirt_fr_01_l_MSTR, loc=<0.2 0 0>, rot=<0 0 0 1>}
 dyn_skirt_fr_02_l {parent=dyn_skirt_fr_01_l, loc=<10.8609 0 0>, rot=<0 0 0 1>}
 dyn_beltstrap_MSTR {parent=pelvis, loc=<6.9257 6.54 4.3326>, rot=<0 0.7071 0.7071 0>}
 dyn_belt_01_r_MSTR {parent=dyn_beltstrap_MSTR, loc=<0.0403 -2.4221 1.8384>, rot=<-0.1123 0.1159 -0.3607 -0.9186>}
 dyn_belt_01_r {parent=dyn_belt_01_r_MSTR, loc=<1.4977 0 0>, rot=<0 0 0 1>}
 dyn_belt_02_r {parent=dyn_belt_01_r, loc=<6.9038 0 0>, rot=<0.2422 -0.0929 0.2322 -0.9374>}
 dyn_belt_03_r {parent=dyn_belt_02_r, loc=<9.5181 0 0>, rot=<0 0 0 1>}
 dyn_belt_04_r {parent=dyn_belt_03_r, loc=<9.5181 0 0>, rot=<0 0 0 1>}
 dyn_belt_05_r {parent=dyn_belt_04_r, loc=<9.5181 0 0>, rot=<0 0 0 1>}
 dyn_skirt_fr_01_r_MSTR {parent=dyn_beltstrap_MSTR, loc=<1.9286 -0.5926 1.1932>, rot=<0.088 0.1041 -0.034 -0.9901>}
 dyn_skirt_fr_01_r {parent=dyn_skirt_fr_01_r_MSTR, loc=<0.1995 0 0.0001>, rot=<0 0 0 1>}
 dyn_skirt_fr_02_r {parent=dyn_skirt_fr_01_r, loc=<10.861 0 0>, rot=<0 0 0 1>}
 dyn_skirt_fr_mid_l_MSTR {parent=pelvis, loc=<4.0319 4.4965 -10.429>, rot=<-0.0323 -0.8914 -0.43 -0.1396>}
 dyn_skirt_fr_mid_l_01 {parent=dyn_skirt_fr_mid_l_MSTR, loc=<0.2 0 0>, rot=<0 0 0 1>}
 dyn_skirt_fr_mid_l_02 {parent=dyn_skirt_fr_mid_l_01, loc=<8.7102 0 0>, rot=<0 0 0 1>}
 dyn_skirt_01_r_MSTR {parent=pelvis, loc=<2.0847 -1.3537 14.0176>, rot=<0.1535 0.0166 0.988 -0.0057>}
 dyn_skirt_01_r {parent=dyn_skirt_01_r_MSTR, loc=<0.1995 0 0.0002>, rot=<0 0 0 1>}
 dyn_skirt_02_r {parent=dyn_skirt_01_r, loc=<6.1888 0 0>, rot=<0 0 0 1>}
 dyn_skirt_fr_mid_r_MSTR {parent=pelvis, loc=<4.0315 4.4965 10.429>, rot=<0.1396 0.43 0.8914 0.0323>}
 dyn_skirt_fr_mid_r_01 {parent=dyn_skirt_fr_mid_r_MSTR, loc=<0.2001 0 0>, rot=<0 0 0 1>}
 dyn_skirt_fr_mid_r_02 {parent=dyn_skirt_fr_mid_r_01, loc=<8.7098 0 0.0001>, rot=<0 0 0 1>}
]
variables [CrouchSprintingAlpha {type=double, default=0}]
