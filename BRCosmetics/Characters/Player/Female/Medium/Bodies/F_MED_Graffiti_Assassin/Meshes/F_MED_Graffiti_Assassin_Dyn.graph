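# Physics asset for secondary motion on the female medium Graffiti Assassin body, covering its belt, skirt panels and gun.
# Tiny MSTR bodies root each chain to the pelvis, mostly with free rotation, with the belt routed through a beltstrap root.
# Capsule chains below them use limited swing and twist, alongside larger pelvis and thigh collision shapes.
# Collar roots have no child bodies, and the gun body has no constraint.
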PhysicsAsset /BRCosmetics/Characters/Player/Female/Medium/Bodies/F_MED_Graffiti_Assassin/Meshes/F_MED_Graffiti_Assassin_Dyn
bodies {
 dyn_belt_01_r_MSTR [{shapeType=Capsule, length=0.505, radius=0.505, rotation=<rot 0 -90 0>}]
 dyn_belt_02_r [{shapeType=Capsule, length=8.1534, radius=2.1203, shapeName=dyn_belt_02_r_capsule, center=<5.6512 0 1.2839>, rotation=<rot 90 0 0>}]
 dyn_belt_03_r [{shapeType=Capsule, length=5.423, radius=2.55, shapeName=dyn_belt_03_r_capsule, center=<4.196 -0.0011 0.3635>, rotation=<rot 90 0 0>}]
 dyn_belt_04_r [{shapeType=Capsule, length=5.3991, radius=2.6573, shapeName=dyn_belt_03_r_capsule, center=<4.196 -0.0011 0.2599>, rotation=<rot 90 0 0>}]
 dyn_belt_05_r [{shapeType=Capsule, length=6.1788, radius=2.4742, shapeName=dyn_belt_03_r_capsule, center=<4.196 -0.0011 -0.0075>, rotation=<rot 90 0 0>}]
 dyn_beltstrap_MSTR [{shapeType=Capsule, length=0.505, radius=0.505, rotation=<rot 0 -90 0>}]
 dyn_collar_01_l_MSTR [{shapeType=Capsule, length=0.505, radius=0.505, rotation=<rot 0 -90 0>}]
 dyn_collar_01_r_MSTR [{shapeType=Capsule, length=0.505, radius=0.505, rotation=<rot 0 -90 0>}]
 dyn_gun [{
   shapeType=Capsule
   length=8.5
   radius=3.75
   shapeName=dyn_gun_capsule
   center=<-4.4803 2.4275 2.5081>
   rotation=<rot -11.5782 8.113 100.0532>
  }]
 dyn_skirt_01_l [{shapeType=Capsule, length=10.7044, radius=1.2086, shapeName=dyn_skirt_01_r_capsule, center=<0.5494 -0.3162 -0.2394>, rotation=<rot 0 0 90>}]
 dyn_skirt_01_l_MSTR [{shapeType=Capsule, length=0.505, radius=0.505, rotation=<rot 0 -90 0>}]
 dyn_skirt_01_r [{shapeType=Capsule, length=10.7044, radius=1.2086, shapeName=dyn_skirt_01_r_capsule, center=<0.5494 -0.3162 -0.2394>, rotation=<rot 0 0 90>}]
 dyn_skirt_01_r_MSTR [{shapeType=Capsule, length=0.505, radius=0.505, rotation=<rot 0 -90 0>}]
 dyn_skirt_02_l [{shapeType=Capsule, length=6.2771, radius=2.4623, shapeName=dyn_skirt_02_r_capsule, center=<5.2561 1.597 -0.8918>, rotation=<rot 0 0 90>}]
 dyn_skirt_02_r [{shapeType=Capsule, length=5.7771, radius=2.9623, shapeName=dyn_skirt_02_r_capsule, center=<5.2561 0 0>, rotation=<rot 0 0 90>}]
 dyn_skirt_bk_01_l [{
   shapeType=Capsule
   length=7.7668
   radius=3.4881
   shapeName=dyn_skirt_bk_01_r_capsule
   center=<8.2224 -0.0529 1.8183>
   rotation=<rot -0.0001 0 100.3058>
  }]
 dyn_skirt_bk_01_l_MSTR [{shapeType=Capsule, length=0.505, radius=0.505, rotation=<rot 0 -90 0>}]
 dyn_skirt_bk_01_r [{
   shapeType=Capsule
   length=7.7668
   radius=3.4881
   shapeName=dyn_skirt_bk_01_r_capsule
   center=<8.2224 0.0321 1.8153>
   rotation=<rot 0 0 83.7764>
  }]
 dyn_skirt_bk_01_r_MSTR [{shapeType=Capsule, length=0.505, radius=0.505, rotation=<rot 0 -90 0>}]
 dyn_skirt_fr_01_l [{shapeType=Capsule, length=6.7216, radius=1.204, shapeName=dyn_skirt_fr_mid_r_01_capsule, center=<5.3008 0 0>, rotation=<rot 0 0 90>}]
 dyn_skirt_fr_01_l_MSTR [{shapeType=Capsule, length=0.505, radius=0.505, rotation=<rot 0 -90 0>}]
 dyn_skirt_fr_01_r [{shapeType=Capsule, length=7.987, radius=3.0153, shapeName=dyn_skirt_fr_01_r_capsule, center=<8.7799 0 -3.0151>, rotation=<rot 0 0 90>}]
 dyn_skirt_fr_01_r_MSTR [{shapeType=Capsule, length=0.505, radius=0.505, rotation=<rot 0 -90 0>}]
 dyn_skirt_fr_02_l [{shapeType=Capsule, length=12.5463, radius=3.2123, shapeName=dyn_skirt_fr_02_r_capsule, center=<4.0726 0 1.775>, rotation=<rot 0 0 90>}]
 dyn_skirt_fr_02_r [{shapeType=Capsule, length=15.5463, radius=2.2298, shapeName=dyn_skirt_fr_02_r_capsule, center=<3.4455 0 -0.2128>, rotation=<rot 0 0 90>}]
 dyn_skirt_fr_mid_l_01 [{shapeType=Capsule, length=6.7216, radius=1.204, shapeName=dyn_skirt_fr_mid_r_01_capsule, center=<3.2655 0 0>, rotation=<rot 0 0 90>}]
 dyn_skirt_fr_mid_l_02 [{
   shapeType=Capsule
   length=7.1765
   radius=3.2123
   shapeName=dyn_skirt_fr_mid_r_02_capsule
   center=<5.0609 0.4282 -0.0603>
   rotation=<rot 0.1588 1.4088 -88.0045>
  }]
 dyn_skirt_fr_mid_l_MSTR [{shapeType=Capsule, length=0.505, radius=0.505, rotation=<rot 0 -90 0>}]
 dyn_skirt_fr_mid_r_01 [{
   shapeType=Capsule
   length=7.9716
   radius=3.204
   shapeName=dyn_skirt_fr_mid_r_01_capsule
   center=<7.1385 0 -0.7767>
   rotation=<rot 0.0002 -0.0002 90.0001>
  }]
 dyn_skirt_fr_mid_r_02 [{
   shapeType=Capsule
   length=14.5116
   radius=2.3772
   shapeName=dyn_skirt_fr_mid_r_02_capsule
   center=<6.3559 -0.1211 0.6021>
   rotation=<rot -0.4005 5.7452 -86.02>
  }]
 dyn_skirt_fr_mid_r_MSTR [{shapeType=Capsule, length=0.505, radius=0.505, rotation=<rot 0 -90 0>}]
 pelvis [
  {shapeType=Capsule, length=9.9209, radius=8.5589, center=<0 -1.4951 -0.0001>}
  {shapeType=Capsule, length=12.1709, radius=9.8714, center=<-5.3513 -0.9055 0>}
 ]
 thigh_l [{shapeType=Capsule, length=33.5459, radius=7.5, center=<-15.7163 0.8944 0.2617>, rotation=<rot 85.9603 179.9996 171.9918>}]
 thigh_r [{shapeType=Capsule, length=26.2179, radius=7.9807, center=<19.3904 0.2291 0.3191>, rotation=<rot 85.9552 180 180>}]
 thigh_twist_01_r [{
   shapeType=Capsule
   length=0
   radius=8.8276
   shapeName=thigh_twist_01_r_capsule
   center=<-18.0012 0.5199 -1.4803>
   rotation=<rot 0 -0.0001 40>
  }]
}
constraints [
 {bone1Name=dyn_belt_02_r, bone2Name=dyn_belt_01_r_MSTR, swing1LimitDegrees=60, swing2LimitDegrees=60, twistLimitDegrees=5}
 {bone1Name=dyn_belt_03_r, bone2Name=dyn_belt_02_r, swing1LimitDegrees=45, swing2LimitDegrees=45, twistLimitDegrees=5}
 {bone1Name=dyn_belt_04_r, bone2Name=dyn_belt_03_r, swing1LimitDegrees=45, swing2LimitDegrees=45, twistLimitDegrees=5}
 {bone1Name=dyn_belt_05_r, bone2Name=dyn_belt_04_r, swing1LimitDegrees=45, swing2LimitDegrees=45, twistLimitDegrees=5}
 {
  bone1Name=dyn_skirt_fr_01_r_MSTR
  bone2Name=pelvis
  swing1LimitDegrees=3
  swing1Motion=Locked
  swing2LimitDegrees=3
  swing2Motion=Locked
  twistLimitDegrees=3
  twistMotion=Locked
 }
 {
  bone1Name=dyn_collar_01_l_MSTR
  bone2Name=pelvis
  swing1LimitDegrees=45
  swing1Motion=Free
  swing2LimitDegrees=45
  swing2Motion=Free
  twistLimitDegrees=45
  twistMotion=Free
 }
 {
  bone1Name=dyn_collar_01_r_MSTR
  bone2Name=pelvis
  swing1LimitDegrees=45
  swing1Motion=Free
  swing2LimitDegrees=45
  swing2Motion=Free
  twistLimitDegrees=45
  twistMotion=Free
 }
 {
  bone1Name=thigh_l
  bone2Name=pelvis
  swing1LimitDegrees=45
  swing1Motion=Free
  swing2LimitDegrees=45
  swing2Motion=Free
  twistLimitDegrees=45
  twistMotion=Free
 }
 {
  bone1Name=thigh_r
  bone2Name=pelvis
  swing1LimitDegrees=45
  swing1Motion=Free
  swing2LimitDegrees=45
  swing2Motion=Free
  twistLimitDegrees=45
  twistMotion=Free
 }
 {
  bone1Name=dyn_skirt_bk_01_l_MSTR
  bone2Name=pelvis
  swing1LimitDegrees=45
  swing1Motion=Free
  swing2LimitDegrees=45
  swing2Motion=Free
  twistLimitDegrees=45
  twistMotion=Free
 }
 {
  bone1Name=dyn_skirt_bk_01_r_MSTR
  bone2Name=pelvis
  swing1LimitDegrees=45
  swing1Motion=Free
  swing2LimitDegrees=45
  swing2Motion=Free
  twistLimitDegrees=45
  twistMotion=Free
 }
 {
  bone1Name=dyn_belt_01_r_MSTR
  bone2Name=dyn_beltstrap_MSTR
  swing1LimitDegrees=45
  swing1Motion=Free
  swing2LimitDegrees=45
  swing2Motion=Free
  twistLimitDegrees=45
  twistMotion=Free
 }
 {
  bone1Name=dyn_skirt_01_r_MSTR
  bone2Name=pelvis
  swing1LimitDegrees=2
  swing1Motion=Locked
  swing2LimitDegrees=2
  swing2Motion=Locked
  twistLimitDegrees=2
  twistMotion=Locked
 }
 {bone1Name=dyn_skirt_01_r, bone2Name=dyn_skirt_01_r_MSTR, swing1LimitDegrees=10, swing2LimitDegrees=30, twistLimitDegrees=5}
 {bone1Name=dyn_skirt_02_r, bone2Name=dyn_skirt_01_r, swing1LimitDegrees=15, swing2LimitDegrees=45, twistLimitDegrees=5}
 {
  bone1Name=dyn_skirt_fr_01_l_MSTR
  bone2Name=pelvis
  swing1LimitDegrees=45
  swing1Motion=Free
  swing2LimitDegrees=45
  swing2Motion=Free
  twistLimitDegrees=45
  twistMotion=Free
 }
 {bone1Name=dyn_skirt_fr_01_l, bone2Name=dyn_skirt_fr_01_l_MSTR, swing1LimitDegrees=5, swing2LimitDegrees=45, twistLimitDegrees=20}
 {bone1Name=dyn_skirt_fr_02_l, bone2Name=dyn_skirt_fr_01_l, swing1LimitDegrees=15, swing2LimitDegrees=45, twistLimitDegrees=5}
 {
  bone1Name=dyn_skirt_fr_mid_r_MSTR
  bone2Name=pelvis
  swing1LimitDegrees=45
  swing1Motion=Free
  swing2LimitDegrees=45
  swing2Motion=Free
  twistLimitDegrees=45
  twistMotion=Free
 }
 {bone1Name=dyn_skirt_fr_mid_r_01, bone2Name=dyn_skirt_fr_mid_r_MSTR, swing1LimitDegrees=15, swing2LimitDegrees=45, twistLimitDegrees=5}
 {bone1Name=dyn_skirt_fr_mid_r_02, bone2Name=dyn_skirt_fr_mid_r_01, swing1LimitDegrees=10, swing2LimitDegrees=45, twistLimitDegrees=5}
 {bone1Name=dyn_skirt_fr_mid_l_02, bone2Name=dyn_skirt_fr_mid_l_01, swing1LimitDegrees=15, swing2LimitDegrees=45, twistLimitDegrees=5}
 {
  bone1Name=dyn_skirt_fr_mid_l_01
  bone2Name=dyn_skirt_fr_mid_l_MSTR
  swing1LimitDegrees=1
  swing1Motion=Locked
  swing2LimitDegrees=30
  twistLimitDegrees=5
  twistMotion=Locked
 }
 {
  bone1Name=dyn_skirt_fr_mid_l_MSTR
  bone2Name=pelvis
  swing1LimitDegrees=45
  swing1Motion=Free
  swing2LimitDegrees=45
  swing2Motion=Free
  twistLimitDegrees=45
  twistMotion=Free
 }
 {
  bone1Name=dyn_beltstrap_MSTR
  bone2Name=pelvis
  swing1LimitDegrees=45
  swing1Motion=Free
  swing2LimitDegrees=45
  swing2Motion=Free
  twistLimitDegrees=45
  twistMotion=Free
 }
 {bone1Name=dyn_skirt_bk_01_r, bone2Name=dyn_skirt_bk_01_r_MSTR, swing1LimitDegrees=5, swing2LimitDegrees=25, twistLimitDegrees=5}
 {bone1Name=dyn_skirt_fr_01_r, bone2Name=dyn_skirt_fr_01_r_MSTR, swing1LimitDegrees=2, swing2LimitDegrees=45, twistLimitDegrees=2}
 {bone1Name=dyn_skirt_fr_02_r, bone2Name=dyn_skirt_fr_01_r, swing1LimitDegrees=5, swing2LimitDegrees=45, twistLimitDegrees=2}
 {bone1Name=dyn_skirt_bk_01_l, bone2Name=dyn_skirt_bk_01_l_MSTR, swing1LimitDegrees=5, swing2LimitDegrees=25, twistLimitDegrees=5}
 {
  bone1Name=dyn_skirt_01_l_MSTR
  bone2Name=pelvis
  swing1LimitDegrees=45
  swing1Motion=Free
  swing2LimitDegrees=45
  swing2Motion=Free
  twistLimitDegrees=45
  twistMotion=Free
 }
 {bone1Name=dyn_skirt_01_l, bone2Name=dyn_skirt_01_l_MSTR, swing1LimitDegrees=25, swing2LimitDegrees=30, twistLimitDegrees=5}
 {bone1Name=dyn_skirt_02_l, bone2Name=dyn_skirt_01_l, swing1LimitDegrees=15, swing2LimitDegrees=45, twistLimitDegrees=5}
 {
  bone1Name=thigh_twist_01_r
  bone2Name=thigh_r
  swing1LimitDegrees=45
  swing1Motion=Free
  swing2LimitDegrees=45
  swing2Motion=Free
  twistLimitDegrees=45
  twistMotion=Free
 }
]
