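# Post-pose AnimBlueprint for the F_MED_Graffiti_Assassin body that adds secondary motion to belts, pouches, skirt, collar and knife.
# It copies the parent mesh pose, then applies Control Rig, crouch bone offsets, weighted constraints, RigidBody, AnimDynamics and skirt corrections.
# EventGraph derives crouch, crouch-sprint and thigh-angle alphas.
# Its computed gravity_override, Crouch Walk and left-thigh values are unused in AnimGraph, and one belt transform node is unconnected.
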AnimBlueprint /BRCosmetics/Characters/Player/Female/Medium/Bodies/F_MED_Graffiti_Assassin/Meshes/F_MED_Graffiti_Assassin_AnimBP
target_skeleton=&/F_MED_Graffiti_Assassin_Skeleton
graphs [
 AnimGraph {nodes [
   "Transform (Modify) Bone - Bone: dyn_belt_01_r"(Alpha=Get thigh rotation_r() -> (thigh rotation_r:double), Translation="-5.693,0.890,1.021", Rotation="8.653,90.212,121.466", Scale="1,1,1") {
    rotationMode=BMM_Additive
    rotationSpace=BCS_BoneSpace
    translationMode=BMM_Additive
    translationSpace=BCS_BoneSpace
    alphaBoolBlend {blendOption=Linear}
    alphaScaleBiasClamp {bInterpResult=true, clampMax=0.1, interpSpeedDecreasing=5, interpSpeedIncreasing=7}
    boneToModify {boneName=dyn_belt_01_r}
    rotation=<rot 8.653 90.212 121.466>
    translation=<-5.693 0.89 1.021>
   }
   n1 = Control Rig!(
    Source=Copy Pose From Mesh() {bUseAttachedParent=true}
    Alpha
    CrouchSprintingAlpha:double=Get CrouchSprintingAnim() -> (CrouchSprintingAnim:double)
   ) -> (Pose) {
    alphaBoolBlend {blendOption=Linear}
    controlRigAssetReference {blueprintRigClass=&/F_MED_Graffiti_Assassin_CtrlRig.F_MED_Graffiti_Assassin_CtrlRig_C}
   }
   n2 = Get crouch alpha() -> (crouch alpha:double)
   n3 = "Transform (Modify) Bone - Bone: dyn_pouch_l_A_MSTR"(
    ComponentPose=Local To Component(LocalPose=@n1)
    Alpha=@n2
    Translation="-0.215,-4.715,-1.146"
    Rotation="-0.881,-13.211,-54.635"
    Scale="1,1,1"
   ) {
    rotationMode=BMM_Additive
    rotationSpace=BCS_BoneSpace
    translationMode=BMM_Additive
    translationSpace=BCS_BoneSpace
    alphaBoolBlend {blendOption=Linear}
    alphaScaleBiasClamp {bInterpResult=true}
    boneToModify {boneName=dyn_pouch_l_A_MSTR}
    rotation=<rot -0.881 -13.211 -54.635>
    translation=<-0.215 -4.715 -1.146>
   }
   n4 = "Transform (Modify) Bone - Bone: dyn_pouch_l_B_MSTR"(ComponentPose=@n3, Alpha=@n2, Translation="-1.023,-3.273,-1.951", Rotation="-0.743,-10.557,-54.407", Scale="1,1,1") {
    rotationMode=BMM_Additive
    rotationSpace=BCS_BoneSpace
    translationMode=BMM_Additive
    translationSpace=BCS_BoneSpace
    alphaBoolBlend {blendOption=Linear}
    alphaScaleBiasClamp {bInterpResult=true}
    boneToModify {boneName=dyn_pouch_l_B_MSTR}
    rotation=<rot -0.743 -10.5571 -54.4069>
    translation=<-1.023 -3.273 -1.951>
   }
   n5 = "Transform (Modify) Bone - Bone: dyn_belt_01_r_MSTR"(
    ComponentPose=@n4
    Alpha=Get Thigh_Crouch_R() -> (Thigh_Crouch_R:double)
    Translation="2.416,2.752,3.197"
    Rotation="-30.724,-61.740,109.828"
    Scale="1,1,1"
   ) {
    rotationMode=BMM_Additive
    rotationSpace=BCS_BoneSpace
    translationMode=BMM_Additive
    translationSpace=BCS_BoneSpace
    alphaBoolBlend {blendOption=Linear}
    alphaScaleBiasClamp {bInterpResult=true, clampMax=0.1, interpSpeedDecreasing=1, interpSpeedIncreasing=1}
    boneToModify {boneName=dyn_belt_01_r_MSTR}
    rotation=<rot -30.724 -61.74 109.828>
    translation=<2.416 2.752 3.197>
   }
   n6 = Get crouch alpha()
   n7 = "Constraint - dyn_belt_02_r (PPPPPPPP)"!(
    ComponentPose=@n5
    Alpha=@n6
    ConstraintWeights_0=0.0291
    ConstraintWeights_1:float=0.018
    ConstraintWeights_2:float=0.1035
    ConstraintWeights_3:float=0.3631
    ConstraintWeights_4:float=0.085
    ConstraintWeights_5:float=0.2717
    ConstraintWeights_6:float=0.0436
    ConstraintWeights_7:float=0.086
   ) -> (Pose) {
    alphaBoolBlend {blendOption=Linear}
    boneToModify {boneName=dyn_belt_02_r}
    constraintSetup [
     {targetBone {boneName=pelvis}}
     {targetBone {boneName=dyn_skirt_01_r}}
     {targetBone {boneName=dyn_skirt_fr_02_r}}
     {targetBone {boneName=dyn_skirt_fr_01_r}}
     {targetBone {boneName=dyn_skirt_fr_01_l}}
     {targetBone {boneName=dyn_skirt_fr_mid_r_01}}
     {targetBone {boneName=dyn_skirt_fr_mid_r_02}}
     {targetBone {boneName=dyn_skirt_fr_01_r_MSTR}}
    ]
    constraintWeights [1,1,1,1,1,1,1,1]
   }
   n8 = "Constraint - dyn_belt_03_r (PPPPPPP)"!(
    ComponentPose=@n7
    Alpha=@n6
    ConstraintWeights_0=0.0062
    ConstraintWeights_1:float=0.0118
    ConstraintWeights_2:float=0.5121
    ConstraintWeights_3:float=0.1749
    ConstraintWeights_4:float=0.02
    ConstraintWeights_5:float=0.0831
    ConstraintWeights_6:float=0.1919
   ) -> (Pose) {
    alphaBoolBlend {blendOption=Linear}
    boneToModify {boneName=dyn_belt_03_r}
    constraintSetup [
     {targetBone {boneName=dyn_skirt_01_r}}
     {targetBone {boneName=dyn_skirt_02_r}}
     {targetBone {boneName=dyn_skirt_fr_02_r}}
     {targetBone {boneName=dyn_skirt_fr_01_r}}
     {targetBone {boneName=dyn_skirt_fr_01_l}}
     {targetBone {boneName=dyn_skirt_fr_mid_r_01}}
     {targetBone {boneName=dyn_skirt_fr_mid_r_02}}
    ]
    constraintWeights [1,1,1,1,1,1,1]
   }
   n9 = RigidBody(ComponentPose=@n8) {
    bClampLinearTranslationLimitToRefPose=true
    overridePhysicsAsset=&/F_MED_Graffiti_Assassin_Dyn
    simulationSpace=BaseBoneSpace
    alphaBoolBlend {blendOption=Linear}
    baseBoneRef {boneName=pelvis}
    componentAppliedLinearAccClamp=<3000 3000 3000>
    componentLinearAccScale=<0.3 0.3 0.3>
    componentLinearVelScale=<0.2 0.2 0.2>
    overrideWorldGravity {z=-980}
   }
   n10 = Anim Dynamics - Bone: dyn_collar_01_l(ComponentPose=@n9, Alpha=1, ExternalForce="0,0,0") {
    angularDampingOverride=0.99
    angularSpringConstant=100
    bOverrideAngularDamping=true
    alphaBoolBlend {blendOption=Linear}
    boundBone {boneName=dyn_collar_01_l}
    componentAppliedLinearAccClamp=<8000 8000 8000>
    componentLinearAccScale=<0.6 0.6 0.6>
    componentLinearVelScale=<0.6 0.6 0.6>
    physicsBodyDefinitions [{
      boundBone {boneName=dyn_collar_01_l}
      boxExtents=<20 20 15>
      constraintSetup {angularLimitsMax=<0 0 5>, angularLimitsMin=<0 -15 -5>, angularTarget=<0 0 0>, linearAxesMax=<0 0 2.5>}
      localJointOffset=<0.5 0 0>
     }]
    relativeSpaceBone {boneName=root}
   }
   n11 = Anim Dynamics - Bone: dyn_collar_01_r(ComponentPose=@n10, Alpha=1, ExternalForce="0,0,0") {
    angularDampingOverride=0.99
    angularSpringConstant=100
    bOverrideAngularDamping=true
    alphaBoolBlend {blendOption=Linear}
    boundBone {boneName=dyn_collar_01_r}
    componentAppliedLinearAccClamp=<8000 8000 8000>
    componentLinearAccScale=<0.6 0.6 0.6>
    componentLinearVelScale=<0.6 0.6 0.6>
    physicsBodyDefinitions [{
      boundBone {boneName=dyn_collar_01_r}
      boxExtents=<20 20 15>
      constraintSetup {angularLimitsMax=<0 0 5>, angularLimitsMin=<0 -15 -5>, angularTarget=<0 0 0>, linearAxesMax=<0 0 2.5>}
      localJointOffset=<0.5 0 0>
     }]
    relativeSpaceBone {boneName=root}
   }
   n12 = Anim Dynamics - Bone: dyn_pouch_l_A(ComponentPose=@n11, Alpha=1, ExternalForce="0,0,0") {
    angularDampingOverride=1
    angularSpringConstant=300
    bAngularSpring=true
    bOverrideAngularDamping=true
    simulationSpace=BoneRelative
    alphaBoolBlend {blendOption=Linear}
    boundBone {boneName=dyn_pouch_l_A}
    componentAppliedLinearAccClamp=<5000 5000 5000>
    componentLinearAccScale=<0.5 0.3 0.3>
    componentLinearVelScale=<0.5 0.3 0.3>
    physicsBodyDefinitions [{
      boundBone {boneName=dyn_pouch_l_A}
      boxExtents=<10 4 4>
      constraintSetup {angularLimitsMax=<5 5 5>, angularLimitsMin=<-5 -5 -5>, angularTarget=<0 0 0>}
      localJointOffset=<3 0 0>
     }]
    relativeSpaceBone {boneName=pelvis}
   }
   n13 = Anim Dynamics - Bone: dyn_pouch_l_B(ComponentPose=@n12, Alpha=1, ExternalForce="0,0,0") {
    angularDampingOverride=1
    angularSpringConstant=300
    bAngularSpring=true
    bOverrideAngularDamping=true
    simulationSpace=BoneRelative
    alphaBoolBlend {blendOption=Linear}
    boundBone {boneName=dyn_pouch_l_B}
    componentAppliedLinearAccClamp=<5000 5000 5000>
    componentLinearAccScale=<0.4 0.3 0.3>
    componentLinearVelScale=<0.4 0.3 0.3>
    physicsBodyDefinitions [{
      boundBone {boneName=dyn_pouch_l_B}
      boxExtents=<15 10 10>
      constraintSetup {angularLimitsMax=<5 5 5>, angularLimitsMin=<-5 -5 -5>, angularTarget=<0 0 0>}
      localJointOffset=<3 0 0>
     }]
    relativeSpaceBone {boneName=pelvis}
   }
   n14 = Anim Dynamics - Bone: dyn_knife_01_l(ComponentPose=@n13, Alpha=1, ExternalForce="0,0,0") {
    angularDampingOverride=0.8
    angularSpringConstant=300
    bAngularSpring=true
    bOverrideAngularDamping=true
    bUseGravityOverride=true
    simulationSpace=BoneRelative
    alphaBoolBlend {blendOption=Linear}
    boundBone {boneName=dyn_knife_01_l}
    componentAppliedLinearAccClamp=<5000 5000 5000>
    componentLinearAccScale=<1 1 1>
    componentLinearVelScale=<1 1 1>
    gravityOverride {z=-900}
    physicsBodyDefinitions [{
      boundBone {boneName=dyn_knife_01_l}
      boxExtents=<7 7 20>
      constraintSetup {angularLimitsMax=<10 10 5>, angularLimitsMin=<-10 -10 -5>, angularTarget=<0 0 -20>}
      localJointOffset=<1.5 1 0>
     }]
    relativeSpaceBone {boneName=pelvis}
   }
   n15 = "Constraint - dyn_skirt_fr_mid_l_01 (PP)"!(ComponentPose=@n14, Alpha=1, ConstraintWeights_0=0.5, ConstraintWeights_1:float=0.5) -> (Pose) {
    alphaBoolBlend {blendOption=Linear}
    boneToModify {boneName=dyn_skirt_fr_mid_l_01}
    constraintSetup [{targetBone {boneName=thigh_l}}, {targetBone {boneName=dyn_skirt_fr_01_l}}]
    constraintWeights [1,1]
   }
   n16 = "Constraint - dyn_skirt_fr_mid_r_01 (PP)"!(ComponentPose=@n15, Alpha=1, ConstraintWeights_0=0.5, ConstraintWeights_1:float=0.5) -> (Pose) {
    alphaBoolBlend {blendOption=Linear}
    boneToModify {boneName=dyn_skirt_fr_mid_r_01}
    constraintSetup [{targetBone {boneName=thigh_r}}, {targetBone {boneName=dyn_skirt_fr_01_r}}]
    constraintWeights [1,1]
   }
   n17 = "Transform (Modify) Bone - Bone: dyn_skirt_fr_mid_r_01"!(ComponentPose=@n16, Alpha=1, Rotation="12,0,0") -> (Pose) {
    rotationMode=BMM_Replace
    rotationSpace=BCS_BoneSpace
    translationMode=BMM_Replace
    translationSpace=BCS_BoneSpace
    alphaBoolBlend {blendOption=Linear}
    alphaScaleBiasClamp {bInterpResult=true, interpSpeedDecreasing=40, interpSpeedIncreasing=40}
    boneToModify {boneName=dyn_skirt_fr_mid_r_01}
    rotation {pitch=12}
    translation {z=2}
   }
   n18 = "Transform (Modify) Bone - Bone: dyn_skirt_fr_mid_l_MSTR"!(ComponentPose=@n17, Alpha=1, Rotation="12,-0.000,0.000") -> (Pose) {
    rotationMode=BMM_Replace
    rotationSpace=BCS_BoneSpace
    translationMode=BMM_Replace
    translationSpace=BCS_BoneSpace
    alphaBoolBlend {blendOption=Linear}
    alphaScaleBiasClamp {bInterpResult=true, interpSpeedDecreasing=40, interpSpeedIncreasing=40}
    boneToModify {boneName=dyn_skirt_fr_mid_l_MSTR}
    rotation {pitch=12}
    translation {z=2}
   }
   Output Pose(Result=Component To Local(ComponentPose=@n18)) {layerGroup=None}
  ]}
 EventGraph {nodes [
   Event BlueprintUpdateAnimation()
   > n1 = Sequence!(execute) -> (then_0, then_1, then_2:Exec, then_3:Exec)
   > n2 = Branch(Condition=Get bIsSwimSprinting() -> (bIsSwimSprinting:bool))
   > Set gravity_override(gravity_override="0, 0, 0")
   n3 = Branch(execute=@n2.else, Condition=Get bIsSkyDiving())
   > n4 = Branch(Condition=Get bIsParachuteOpened())
   Set gravity_override(execute=[@n3.else, @n4.then], gravity_override="0, 0,-980")
   n5 = Branch(execute=@n4.else, Condition="float > float"(A=Get PawnVelocityZ(), B=50) {is=title K2Node_CallFunction})
   > Set gravity_override(gravity_override="0, 0,200")
   Set gravity_override(execute=@n5.else, gravity_override="0, 0,-200")
   n6 = GetOwnerSkeletalMeshComponent()
   n7 = BreakRotator(InRot=BreakTransform(InTransform=GetDeltaTransformFromRefPose(self=@n6, BoneName=thigh_r, BaseName=pelvis)).Rotation)
   n8 = OR Boolean!(A=Get bRigidBodyEnabled(), B=Get bAnimDynamicsEnabled() -> (bAnimDynamicsEnabled:bool), C:bool=Get bIsCrouching()) -> (ReturnValue)
   Set thigh rotation_r(
    execute=@n1.then_1
    thigh rotation_r:double=SelectFloat(A=0, B=MapRangeClamped(Value=@n7.Yaw, InRangeA=-30, InRangeB=-65), bPickA=NOT Boolean(A=@n8))
   ) -> (Output_Get:double)
   > Set Thigh_Crouch_R(Thigh_Crouch_R:double=SelectFloat(A=MapRangeClamped(Value=@n7.Yaw, InRangeA=50, InRangeB=-70), B=0, bPickA=Get bIsCrouching())) -> (Output_Get:double)
   n9 = BreakTransform(InTransform=GetDeltaTransformFromRefPose(self=GetOwnerSkeletalMeshComponent(), BoneName=thigh_l, BaseName=pelvis))
   Set thigh_rotation_l(execute=@n1.then_2, thigh_rotation_l:double=MapRangeClamped(Value=BreakRotator(InRot=@n9.Rotation).Yaw, InRangeA=-50, InRangeB=-90)) -> (Output_Get:double)
   Set CrouchSprintingAnim(execute=@n1.then_3, CrouchSprintingAnim:double=SelectFloat(A=1, B=0, bPickA=Get bIsCrouchSprinting())) -> (Output_Get:double)
   n10 = Get bIsCrouching()
   n11 = Get bIsMoving2D()
   > Set Crouch Walk(Crouch Walk:double=SelectFloat(A=1, B=0, bPickA=AND Boolean(A=@n10, B=@n11))) -> (Output_Get:double)
   Event OnBeginCrouching()
   > Set crouch alpha(crouch alpha:double=1) -> (Output_Get:double)
   Event OnEndCrouching()
   > Set crouch alpha(crouch alpha=0)
   PrintString!(execute, InString="To String (Float)"(InDouble=@n7.Yaw), bPrintToScreen, bPrintToLog, TextColor, Duration, Key) -> (then)
  ]}
]
variables [
 crouch alpha {type=number}
 Crouch Walk {type=number}
 CrouchSprintingAnim {type=number}
 gravity_override {type=Vector}
 thigh rotation_r {type=number}
 Thigh_Crouch_R {type=number}
 thigh_rotation_l {type=number}
]
