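# Control Rig for the female Graffiti Assassin face accessory that applies corrective offsets to dynamic hair bones.
# Front End and ADS booleans select one of three branches: fixed front-end offsets, ADS correctives or default correctives.
# The latter two use spherical pose readers on head orientation relative to the pelvis to weight most offsets, with some always full weight.
# A shared function adds rotation, translation and scale offsets to each bone's current local transform.
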ControlRigBlueprint /BRCosmetics/Characters/Player/Female/Medium/Bodies/F_MED_Graffiti_Assassin/Meshes/Parts/F_MED_Graffiti_Assassin_FaceAcc_CtrlRig
graphs [
 ContainedGraph {nodes [
   n1 = RigVMFunctionEntryNode()
   RigUnit_GetTransform = &RigUnit_GetTransform(Space=LocalSpace, bInitial=false, Item=@n1.Joint)
   n2 = &RigVMFunction_MathQuaternionToEuler(RotationOrder=ZYX, Value=@RigUnit_GetTransform.Transform.Rotation)
   > RigUnit_SetTransform = &RigUnit_SetTransform(
    Space=LocalSpace
    Value=(Rotation=(X=0.0132,Y=0.4099,Z=-0.0222,W=0.9118),Translation=(X=23.6045,Y=-6.37,Z=140.1115),Scale3D=(X=1,Y=1,Z=1))
    bInitial=false
    Item=@n1.Joint
    Weight=@n1.Weight
    Value.Translation.Z=&RigVMFunction_MathFloatAdd(B=@n1.Translate Z, A=@RigUnit_GetTransform.Transform.Translation.Z) {template="Add::Execute(in A,in B,out Result)"}.Result
    Value.Scale3D.Z=&RigVMFunction_MathFloatAdd(B=@n1.Scale Z, A=@RigUnit_GetTransform.Transform.Scale3D.Z) {template="Add::Execute(in A,in B,out Result)"}.Result
    Value.Translation.Y=&RigVMFunction_MathFloatAdd(B=@n1.Translate Y, A=@RigUnit_GetTransform.Transform.Translation.Y) {template="Add::Execute(in A,in B,out Result)"}.Result
    Value.Scale3D.Y=&RigVMFunction_MathFloatAdd(B=@n1.Scale Y, A=@RigUnit_GetTransform.Transform.Scale3D.Y) {template="Add::Execute(in A,in B,out Result)"}.Result
    Value.Translation.X=&RigVMFunction_MathFloatAdd(B=@n1.Translate X, A=@RigUnit_GetTransform.Transform.Translation.X) {template="Add::Execute(in A,in B,out Result)"}.Result
    Value.Scale3D.X=&RigVMFunction_MathFloatAdd(B=@n1.Scale X, A=@RigUnit_GetTransform.Transform.Scale3D.X) {template="Add::Execute(in A,in B,out Result)"}.Result
    Value.Rotation=&RigVMFunction_MathQuaternionFromEuler(
     Euler=(X=0,Y=0,Z=0)
     Euler.X=&RigVMFunction_MathFloatAdd(B=@n1.Rotate X, A=@n2.Result.X) {template="Add::Execute(in A,in B,out Result)"}.Result
     Euler.Y=&RigVMFunction_MathFloatAdd(B=@n1.Rotate Y, A=@n2.Result.Y) {template="Add::Execute(in A,in B,out Result)"}.Result
     Euler.Z=&RigVMFunction_MathFloatAdd(B=@n1.Rotate Z, A=@n2.Result.Z) {template="Add::Execute(in A,in B,out Result)"}.Result
    ).Result
   )
   > RigVMFunctionReturnNode()
  ]}
 RigVMFunctionLibrary {nodes [Set Transforms for Bone = RigVMCollapseNode(
    ExecuteContext=()
    Joint=()
    Rotate X=0
    Rotate Y=0
    Rotate Z=0
    Scale X=0
    Scale Y=0
    Scale Z=0
    Translate X=0
    Translate Y=0
    Translate Z=0
    Weight=1
   )]}
 RigVMModel {nodes [
   RigUnit_BeginExecution = &RigUnit_BeginExecution()
   Branch_1_RigVMUnitNode = &RigVMFunction_ControlFlowBranch(ExecuteContext=@RigUnit_BeginExecution.ExecutePin, Condition=RigVMVariableNode() {variable=Front End}.Value)
   Branch_RigVMUnitNode = &RigVMFunction_ControlFlowBranch(ExecuteContext=@Branch_1_RigVMUnitNode.False, Condition=RigVMVariableNode() {variable=ADS}.Value)
   n1 = &RigUnit_SphericalPoseReader(
    ActiveRegionScaleFactors=(PositiveWidth=5,NegativeWidth=5,PositiveHeight=1.5,NegativeHeight=4)
    ActiveRegionSize=0.05
    Debug=(bDrawDebug=false,bDraw2D=false,bDrawLocalAxes=false,DebugScale=25,DebugSegments=20,DebugThickness=0.25)
    DriverAxis=(X=1,Y=0,Z=0)
    DriverItem=(Type=Bone,Name="head")
    FalloffRegionScaleFactors=(PositiveWidth=0.5,NegativeWidth=0.5,PositiveHeight=0.15,NegativeHeight=0.5)
    FalloffSize=0.5
    FlipWidthScaling=true
    OptionalParentItem=(Type=Bone,Name="pelvis")
    RotationOffset=(X=-45,Y=-90,Z=0)
    ExecutePin=@Branch_RigVMUnitNode.False
   )
   > Set Transforms for Bone_1_1_1_1_1_1_5 = RigVMFunctionReferenceNode(
    Joint=(Type=Bone,Name="dyn_hair_fr_01_l")
    Rotate X=0
    Rotate Y=0
    Rotate Z=-5
    Scale X=0
    Scale Y=0
    Scale Z=0
    Translate X=0
    Translate Y=0
    Translate Z=0
    Weight=@n1.OutputParam
   )
   > Set Transforms for Bone_1_1_1_1_1_4 = RigVMFunctionReferenceNode(
    Joint=(Type=Bone,Name="dyn_hair_fr_01_r")
    Rotate X=0
    Rotate Y=0
    Rotate Z=-5
    Scale X=0
    Scale Y=0
    Scale Z=0
    Translate X=0
    Translate Y=0
    Translate Z=0
    Weight=@n1.OutputParam
   )
   > Set Transforms for Bone_1_1_1_1_1_1_5_1 = RigVMFunctionReferenceNode(
    Joint=(Type=Bone,Name="dyn_hair_bk_01_r")
    Rotate X=10
    Rotate Y=-5
    Rotate Z=0
    Scale X=0
    Scale Y=0
    Scale Z=0
    Translate X=0
    Translate Y=0
    Translate Z=0
    Weight=@n1.OutputParam
   )
   > Set Transforms for Bone_1_1_1_1_1_1_5_1_3 = RigVMFunctionReferenceNode(
    Joint=(Type=Bone,Name="dyn_hair_bk_01_l")
    Rotate X=-18
    Rotate Y=7
    Rotate Z=0
    Scale X=0
    Scale Y=0
    Scale Z=0
    Translate X=0
    Translate Y=0
    Translate Z=0
    Weight=1
   )
   > n2 = &RigUnit_SphericalPoseReader(
    ActiveRegionScaleFactors=(PositiveWidth=5,NegativeWidth=5,PositiveHeight=1,NegativeHeight=3.5)
    ActiveRegionSize=0.05
    Debug=(bDrawDebug=false,bDraw2D=false,bDrawLocalAxes=false,DebugScale=25,DebugSegments=20,DebugThickness=0.25)
    DriverAxis=(X=1,Y=0,Z=0)
    DriverItem=(Type=Bone,Name="head")
    FalloffRegionScaleFactors=(PositiveWidth=0.5,NegativeWidth=0.5,PositiveHeight=0.3,NegativeHeight=0.8)
    FalloffSize=0.5
    FlipWidthScaling=true
    OptionalParentItem=(Type=Bone,Name="pelvis")
    RotationOffset=(X=65,Y=-90,Z=0)
   )
   > Set Transforms for Bone_1_1_1_1_1_4_1 = RigVMFunctionReferenceNode(
    Joint=(Type=Bone,Name="dyn_hair_bk_01_l")
    Rotate X=0
    Rotate Y=-20
    Rotate Z=0
    Scale X=0.2
    Scale Y=0
    Scale Z=0
    Translate X=0
    Translate Y=0
    Translate Z=0
    Weight=@n2.OutputParam
   )
   > Set Transforms for Bone_1_1_1_1_1_1_5_2 = RigVMFunctionReferenceNode(
    Joint=(Type=Bone,Name="dyn_hair_bk_02")
    Rotate X=0
    Rotate Y=35
    Rotate Z=0
    Scale X=0
    Scale Y=0
    Scale Z=0
    Translate X=0
    Translate Y=0
    Translate Z=0
    Weight=@n2.OutputParam
   )
   > Set Transforms for Bone_1_1_1_1_1_1_5_1_2 = RigVMFunctionReferenceNode(
    Joint=(Type=Bone,Name="dyn_hair_bk_02_r")
    Rotate X=0
    Rotate Y=35
    Rotate Z=-5
    Scale X=0.25
    Scale Y=0
    Scale Z=0
    Translate X=0
    Translate Y=0
    Translate Z=0
    Weight=@n2.OutputParam
   )
   > Set Transforms for Bone_1_1_1_1_1_1_5_1_2_1 = RigVMFunctionReferenceNode(
    Joint=(Type=Bone,Name="dyn_hair_02_r")
    Rotate X=0
    Rotate Y=0
    Rotate Z=0
    Scale X=0
    Scale Y=0
    Scale Z=0
    Translate X=0
    Translate Y=0
    Translate Z=0.5
    Weight=1
   )
   n3 = &RigUnit_SphericalPoseReader(
    ActiveRegionScaleFactors=(PositiveWidth=5,NegativeWidth=5,PositiveHeight=1,NegativeHeight=5)
    ActiveRegionSize=0.05
    Debug=(bDrawDebug=false,bDraw2D=false,bDrawLocalAxes=false,DebugScale=25,DebugSegments=20,DebugThickness=0.25)
    DriverAxis=(X=1,Y=0,Z=0)
    DriverItem=(Type=Bone,Name="head")
    FalloffRegionScaleFactors=(PositiveWidth=0.5,NegativeWidth=0.5,PositiveHeight=0.5,NegativeHeight=0.5)
    FalloffSize=0.5
    FlipWidthScaling=true
    OptionalParentItem=(Type=Bone,Name="pelvis")
    RotationOffset=(X=-45,Y=-90,Z=0)
    ExecutePin=@Branch_RigVMUnitNode.True
   )
   > Set Transforms for Bone_1_1_1_1_1_1_5_1_2_2 = RigVMFunctionReferenceNode(
    Joint=(Type=Bone,Name="dyn_hair_fr_01_r")
    Rotate X=0
    Rotate Y=-15
    Rotate Z=0
    Scale X=0
    Scale Y=0
    Scale Z=0
    Translate X=0
    Translate Y=0
    Translate Z=0
    Weight=@n3.OutputParam
   )
   > Set Transforms for Bone_1_1_1_1_1_1_5_1_2_1_1 = RigVMFunctionReferenceNode(
    Joint=(Type=Bone,Name="dyn_hair_01_r")
    Rotate X=0
    Rotate Y=0
    Rotate Z=20
    Scale X=0
    Scale Y=0
    Scale Z=0
    Translate X=0
    Translate Y=0
    Translate Z=0
    Weight=@n3.OutputParam
   )
   > Set Transforms for Bone_1_1_1_1_1_1_5_1_3_1 = RigVMFunctionReferenceNode(
    Joint=(Type=Bone,Name="dyn_hair_bk_01_l")
    Rotate X=-30
    Rotate Y=5
    Rotate Z=0
    Scale X=0
    Scale Y=0
    Scale Z=0
    Translate X=0
    Translate Y=0
    Translate Z=0
    Weight=1
   )
   > n4 = &RigUnit_SphericalPoseReader(
    ActiveRegionScaleFactors=(PositiveWidth=5,NegativeWidth=5,PositiveHeight=5,NegativeHeight=1)
    ActiveRegionSize=0.05
    Debug=(bDrawDebug=false,bDraw2D=false,bDrawLocalAxes=false,DebugScale=25,DebugSegments=20,DebugThickness=0.25)
    DriverAxis=(X=1,Y=0,Z=0)
    DriverItem=(Type=Bone,Name="head")
    FalloffRegionScaleFactors=(PositiveWidth=0.5,NegativeWidth=0.5,PositiveHeight=0.2,NegativeHeight=0.2)
    FalloffSize=0.5
    FlipWidthScaling=true
    OptionalParentItem=(Type=Bone,Name="pelvis")
    RotationOffset=(X=0,Y=-85,Z=25)
   )
   > Set Transforms for Bone_1_1_1_1_1_4_1_1 = RigVMFunctionReferenceNode(
    Joint=(Type=Bone,Name="dyn_hair_bk_01_l")
    Rotate X=60
    Rotate Y=0
    Rotate Z=0
    Scale X=0
    Scale Y=0
    Scale Z=0
    Translate X=0
    Translate Y=0
    Translate Z=0
    Weight=@n4.OutputParam
   )
   > Set Transforms for Bone_1_1_1_1_1_1_5_2_1 = RigVMFunctionReferenceNode(
    Joint=(Type=Bone,Name="dyn_hair_bk_02")
    Rotate X=0
    Rotate Y=35
    Rotate Z=0
    Scale X=-0.2
    Scale Y=0
    Scale Z=0
    Translate X=0
    Translate Y=0
    Translate Z=0
    Weight=@n4.OutputParam
   )
   > Set Transforms for Bone_1_1_1_1_1_1_5_1_2_3 = RigVMFunctionReferenceNode(
    Joint=(Type=Bone,Name="dyn_hair_bk_02_r")
    Rotate X=0
    Rotate Y=25
    Rotate Z=0
    Scale X=0
    Scale Y=0
    Scale Z=0
    Translate X=0
    Translate Y=0
    Translate Z=0
    Weight=@n4.OutputParam
   )
   > Set Transforms for Bone_1_1_1_1_1_1_5_1_2_1_2 = RigVMFunctionReferenceNode(
    Joint=(Type=Bone,Name="dyn_hair_02_r")
    Rotate X=0
    Rotate Y=0
    Rotate Z=0
    Scale X=0
    Scale Y=0
    Scale Z=0
    Translate X=0
    Translate Y=0
    Translate Z=0.5
    Weight=1
   )
   Set Transforms for Bone_1_1_1_1_1_4_1_2 = RigVMFunctionReferenceNode(
    Joint=(Type=Bone,Name="dyn_hair_bk_01_l")
    Rotate X=0
    Rotate Y=-15
    Rotate Z=5
    Scale X=0
    Scale Y=0
    Scale Z=0
    Translate X=0
    Translate Y=0
    Translate Z=0
    Weight=1
    ExecuteContext=@Branch_1_RigVMUnitNode.True
   )
   > Set Transforms for Bone_1_1_1_1_1_1_5_2_2 = RigVMFunctionReferenceNode(
    Joint=(Type=Bone,Name="dyn_hair_bk_02")
    Rotate X=0
    Rotate Y=32
    Rotate Z=0
    Scale X=0
    Scale Y=0
    Scale Z=0
    Translate X=0
    Translate Y=0
    Translate Z=0
    Weight=1
   )
   > Set Transforms for Bone_1_1_1_1_1_1_5_1_2_4 = RigVMFunctionReferenceNode(
    Joint=(Type=Bone,Name="dyn_hair_bk_02_r")
    Rotate X=0
    Rotate Y=30
    Rotate Z=-5
    Scale X=0
    Scale Y=0
    Scale Z=0
    Translate X=0
    Translate Y=0
    Translate Z=0
    Weight=1
   )
  ]}
]
hierarchy [
 root {loc=<0 0 0>, rot=<0 0 0 1>}
 pelvis {parent=root, loc=<0 -0.5301 98.0571>, rot=<0.0127 0.707 -0.0127 -0.707>}
 spine_01 {parent=pelvis, loc=<6.1208 -0.9328 0>, rot=<0 0 0.034 -0.9994>}
 spine_02 {parent=spine_01, loc=<6.5996 -0.2095 0>, rot=<0 0 0.0238 -0.9997>}
 spine_03 {parent=spine_02, loc=<6.4914 -0.1237 0>, rot=<0 0 0.0149 -0.9999>}
 spine_04 {parent=spine_03, loc=<6.9598 0.0794 0>, rot=<0 0 -0.0132 -0.9999>}
 spine_05 {parent=spine_04, loc=<6.9987 0 0>, rot=<0 0 -0.0015 -1>}
 clavicle_l {parent=spine_05, loc=<5.4831 -0.1053 -3.2946>, rot=<0.0465 -0.7289 -0.009 -0.6829>}
 upperarm_l {parent=clavicle_l, loc=<9.8658 -1.2136 -0.4105>, rot=<0.0164 -0.3245 -0.0269 -0.9454>}
 clavicle_r {parent=spine_05, loc=<5.4831 -0.1053 3.2946>, rot=<0.7289 0.0465 -0.6829 0.009>}
 upperarm_r {parent=clavicle_r, loc=<-9.8658 1.2136 0.4104>, rot=<0.0164 -0.3245 -0.0269 -0.9454>}
 neck_01 {parent=spine_05, loc=<11.3138 0.1529 0>, rot=<0 0 -0.2003 -0.9797>}
 neck_02 {parent=neck_01, loc=<4.5627 0 0>, rot=<0 0 0.0504 -0.9987>}
 head {parent=neck_02, loc=<5.3914 0 0>, rot=<0 0 0.115 -0.9934>}
 dyn_hair_bk_01 {parent=head, loc=<10.5884 -7.1666 0>, rot=<0.1199 -0.6969 0.6969 -0.1199>}
 dyn_hair_bk_02 {parent=dyn_hair_bk_01, loc=<6.5695 0 0>, rot=<0 -0.1102 0 -0.9939>}
 dyn_hair_bk_03 {parent=dyn_hair_bk_02, loc=<7.6449 0 0>, rot=<0 0 0 1>}
 dyn_hair_bk_04 {parent=dyn_hair_bk_03, loc=<7.2466 0 0>, rot=<0 0 0 1>}
 dyn_hair_bk_01_l {parent=head, loc=<10.3204 -6.2311 -2.861>, rot=<0.0811 -0.9045 0.374 -0.1883>}
 dyn_hair_bk_02_l {parent=dyn_hair_bk_01_l, loc=<6.8242 0 0>, rot=<-0.0357 -0.1634 0.0059 -0.9859>}
 dyn_hair_bk_03_l {parent=dyn_hair_bk_02_l, loc=<5.0303 0 0>, rot=<0 0 0 1>}
 dyn_hair_bk_04_l {parent=dyn_hair_bk_03_l, loc=<8.1235 0 0>, rot=<0 0 0 1>}
 dyn_hair_01_l {parent=head, loc=<10.5152 -1.156 -5.298>, rot=<0.0041 -0.9926 -0.0005 -0.1214>}
 dyn_hair_02_l {parent=dyn_hair_01_l, loc=<6.7006 0 0>, rot=<0 -0.0823 0 -0.9966>}
 dyn_hair_03_l {parent=dyn_hair_02_l, loc=<4.7836 0 0>, rot=<0 0 0 1>}
 dyn_hair_04_l {parent=dyn_hair_03_l, loc=<5.9422 0 0>, rot=<0 0 0 1>}
 dyn_hair_fr_01_l {parent=head, loc=<10.4213 4.7165 -5.2395>, rot=<0.1017 -0.9869 -0.0247 -0.1229>}
 dyn_hair_fr_02_l {parent=dyn_hair_fr_01_l, loc=<6.633 0 0>, rot=<-0.0246 -0.0822 0.002 -0.9963>}
 dyn_hair_fr_03_l {parent=dyn_hair_fr_02_l, loc=<5.8524 0 0>, rot=<0 0 -0.1091 -0.994>}
 dyn_hair_fr_04_l {parent=dyn_hair_fr_03_l, loc=<7.9074 0 0>, rot=<0 0 0 1>}
 dyn_hair_bk_01_r {parent=head, loc=<10.3206 -6.2311 2.861>, rot=<0.1883 -0.374 0.9045 -0.0811>}
 dyn_hair_bk_02_r {parent=dyn_hair_bk_01_r, loc=<6.8248 0 -0.0003>, rot=<0.044 -0.1634 -0.0073 -0.9856>}
 dyn_hair_bk_03_r {parent=dyn_hair_bk_02_r, loc=<5.4447 0 0>, rot=<0 0 0 1>}
 dyn_hair_bk_04_r {parent=dyn_hair_bk_03_r, loc=<7.7087 0 0>, rot=<0 0 0 1>}
 dyn_hair_01_r {parent=head, loc=<10.3065 -1.35 5.317>, rot=<0.1294 -0.1405 0.9802 -0.0528>}
 dyn_hair_02_r {parent=dyn_hair_01_r, loc=<6.7015 0 -0.0002>, rot=<0 -0.0823 0 -0.9966>}
 dyn_hair_03_r {parent=dyn_hair_02_r, loc=<5.653 0 0>, rot=<0 0 0 1>}
 dyn_hair_04_r {parent=dyn_hair_03_r, loc=<5.0724 0 0>, rot=<0 0 0 1>}
 dyn_hair_fr_01_r {parent=head, loc=<10.1767 3.2514 5.275>, rot=<0.1225 0.027 0.9886 -0.0838>}
 dyn_hair_fr_02_r {parent=dyn_hair_fr_01_r, loc=<6.633 0 0>, rot=<0.0286 -0.0809 0.0478 -0.9952>}
 dyn_hair_fr_03_r {parent=dyn_hair_fr_02_r, loc=<5.7119 0 0>, rot=<0 0 0.0421 -0.9991>}
 dyn_hair_fr_04_r {parent=dyn_hair_fr_03_r, loc=<8.0481 0 0>, rot=<0 0 0 1>}
]
variables [
 ADS {type=bool, default=False}
 Front End {type=bool, default=False}
]
