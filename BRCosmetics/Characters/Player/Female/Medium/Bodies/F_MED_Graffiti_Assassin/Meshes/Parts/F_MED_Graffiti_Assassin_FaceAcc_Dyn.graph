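# Physics asset for the face accessory part of the female medium Graffiti Assassin player cosmetic, set up for dynamic hair motion.
# Seven four-bone hair chains (side, front and back) hang from the head as capsule bodies, joined by angular constraints with tight twist and mostly moderate swing limits.
# Capsules on the head, neck, upper spine, clavicles and upper arms form the body volumes around the hair, linked by constraints with free rotation.
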PhysicsAsset /BRCosmetics/Characters/Player/Female/Medium/Bodies/F_MED_Graffiti_Assassin/Meshes/Parts/F_MED_Graffiti_Assassin_FaceAcc_Dyn
bodies {
 clavicle_l [{shapeType=Capsule, length=3.875, radius=7.25, center=<4.1362 -1.0304 -2.0008>, rotation=<rot 86.4466 0 0>}]
 clavicle_r [{shapeType=Capsule, length=5.8438, radius=7.25, center=<-4.1362 1.0304 2.0008>, rotation=<rot 86.4466 0 0>}]
 dyn_hair_01_l [{shapeType=Capsule, length=6.755, radius=2.005, shapeName=dyn_hair_01_l_capsule, center=<3.2573 0 1.2>, rotation=<rot 90 0 0>}]
 dyn_hair_01_r [{shapeType=Capsule, length=3.3207, radius=2.5, shapeName=dyn_hair_01_r_capsule, center=<3.4673 0 1.2>, rotation=<rot 90 0 0>}]
 dyn_hair_02_l [{shapeType=Capsule, length=5.505, radius=2.005, shapeName=dyn_hair_02_l_capsule, center=<2.2348 0 0>, rotation=<rot 90 0 0>}]
 dyn_hair_02_r [{shapeType=Capsule, length=4.1381, radius=1.9939, shapeName=dyn_hair_02_r_capsule, center=<3.5929 0 0>, rotation=<rot 90 0 0>}]
 dyn_hair_03_l [{shapeType=Capsule, length=5.3624, radius=2.1242, shapeName=dyn_hair_03_l_capsule, center=<2.9712 0 0>, rotation=<rot 90 0 0>}]
 dyn_hair_03_r [{shapeType=Capsule, length=1.6497, radius=1.5, shapeName=dyn_hair_03_r_capsule, center=<2.8857 0 0>, rotation=<rot 90 0 0>}]
 dyn_hair_04_l [{
   shapeType=Capsule
   length=3.7682
   radius=1.8154
   shapeName=dyn_hair_04_l_capsule
   center=<3.5924 1.1444 0.6398>
   rotation=<rot 0 -58.9889 -87.2995>
  }]
 dyn_hair_04_r [{shapeType=Capsule, length=4.7188, radius=1.5625, center=<3.4379 1.4237 1.2171>, rotation=<rot 3.0929 -57.9596 56.9599>}]
 dyn_hair_bk_01 [{shapeType=Capsule, length=4.255, radius=1.755, shapeName=dyn_hair_bk_01_capsule, center=<3.6902 0 0.5>, rotation=<rot 90 0 0>}]
 dyn_hair_bk_01_l [{shapeType=Capsule, length=6.005, radius=2.005, shapeName=dyn_hair_bk_01_l_capsule, center=<3.135 0 0.5>, rotation=<rot 90 0 0>}]
 dyn_hair_bk_01_r [{shapeType=Capsule, length=6.8711, radius=1.8669, shapeName=dyn_hair_bk_01_r_capsule, center=<3.4218 0 0.5>, rotation=<rot 90 0 0>}]
 dyn_hair_bk_02 [{shapeType=Capsule, length=7.255, radius=1.505, shapeName=dyn_hair_bk_02_capsule, center=<3.6555 0 0>, rotation=<rot 90 0 0>}]
 dyn_hair_bk_02_l [{shapeType=Capsule, length=2.5637, radius=2.8022, shapeName=dyn_hair_bk_02_l_capsule, center=<2.5445 0 0>, rotation=<rot 90 0 0>}]
 dyn_hair_bk_02_r [{shapeType=Capsule, length=7.4333, radius=2, shapeName=dyn_hair_bk_02_r_capsule, center=<4.255 0 0>, rotation=<rot 90 0 0>}]
 dyn_hair_bk_03 [{shapeType=Capsule, length=4.8409, radius=1.9562, shapeName=dyn_hair_bk_03_capsule, center=<3.5606 0 0>, rotation=<rot 90 0 0>}]
 dyn_hair_bk_03_l [{shapeType=Capsule, length=5.5239, radius=2.4524, shapeName=dyn_hair_bk_03_l_capsule, center=<3.7631 0 0>, rotation=<rot 90 0 0>}]
 dyn_hair_bk_03_r [{shapeType=Capsule, length=3.3404, radius=2, shapeName=dyn_hair_bk_03_r_capsule, center=<3.4027 0 0>, rotation=<rot 90 0 0>}]
 dyn_hair_bk_04 [{
   shapeType=Capsule
   length=4.9564
   radius=2.5063
   shapeName=dyn_hair_bk_04_capsule
   center=<4.3643 -0.4263 -0.7749>
   rotation=<rot 0 -104.8439 -76.566>
  }]
 dyn_hair_bk_04_l [{
   shapeType=Capsule
   length=3.7677
   radius=2.6935
   shapeName=dyn_hair_bk_04_l_capsule
   center=<2.9383 -1.26 -0.2471>
   rotation=<rot -5.9429 -110.8131 -82.3114>
  }]
 dyn_hair_bk_04_r [{shapeType=Capsule, length=2.9688, radius=2.8125, center=<3.1503 0.5981 -0.023>, rotation=<rot 1.1697 -47.6053 98.7367>}]
 dyn_hair_fr_01_l [{shapeType=Capsule, length=6.8726, radius=2.6378, shapeName=dyn_hair_fr_01_l_capsule, center=<4.4612 -0.1854 1.2>, rotation=<rot 90 0 0>}]
 dyn_hair_fr_01_r [{shapeType=Capsule, length=5.755, radius=2.255, shapeName=dyn_hair_fr_01_r_capsule, center=<3.7842 0 1.2>, rotation=<rot 90 0 0>}]
 dyn_hair_fr_02_l [{shapeType=Capsule, length=7.1629, radius=2.3278, shapeName=dyn_hair_fr_02_l_capsule, center=<3.4711 0 0.9579>, rotation=<rot 90 0 0>}]
 dyn_hair_fr_02_r [{shapeType=Capsule, length=6.4885, radius=2.362, shapeName=dyn_hair_fr_02_r_capsule, center=<3.5163 0 0.9579>, rotation=<rot 90 0 0>}]
 dyn_hair_fr_03_l [{shapeType=Capsule, length=3.4689, radius=3.0137, shapeName=dyn_hair_fr_03_l_capsule, center=<3.2626 0 1.7858>, rotation=<rot 90 0 0>}]
 dyn_hair_fr_03_r [
  {shapeType=Capsule, length=3.3125, radius=3.3125, center=<3.9784 0.5081 1.044>, rotation=<rot 90 0 0>}
  {shapeType=Capsule, length=3, radius=2, shapeName=dyn_hair_fr_03_r_capsule, center=<5.0031 3.0165 0>, rotation=<rot 90 0 0>}
 ]
 dyn_hair_fr_04_l [{shapeType=Capsule, length=3.5625, radius=2.3125, center=<1.9135 1.3116 -1.9924>, rotation=<rot 83.5329 27.1716 -0.0001>}]
 dyn_hair_fr_04_r [{shapeType=Capsule, length=2.5, radius=2.3125, center=<1.7938 -1.7526 -2.0031>, rotation=<rot 83.5329 -27.1716 0.0001>}]
 head [
  {shapeType=Capsule, length=0.5, radius=6.25, center=<2.6046 -0.5902 0.0001>, rotation=<rot 90 0 0>}
  {shapeType=Capsule, length=4.25, radius=3, shapeName=head_capsule, center=<2.0745 4.741 0>}
 ]
 neck_01 [{shapeType=Capsule, length=4.4801, radius=6.5, center=<1.158 1.6419 0>, rotation=<rot 90 0 0>}]
 spine_05 [{shapeType=Capsule, length=10.5625, radius=6.0312, center=<7.7449 0.2141 0>}]
 upperarm_l [{shapeType=Capsule, length=10.0625, radius=5.2188, center=<6.552 -0.3397 -0.6841>, rotation=<rot 87.3013 0.5714 2.1564>}]
 upperarm_r [{shapeType=Capsule, length=21.5938, radius=5.2188, center=<-11.1499 0.4674 0.9006>, rotation=<rot 87.3029 0.5714 2.1564>}]
}
constraints [
 {bone1Name=dyn_hair_bk_01, bone2Name=head, swing1LimitDegrees=5, swing2LimitDegrees=30, twistLimitDegrees=5}
 {bone1Name=dyn_hair_bk_02, bone2Name=dyn_hair_bk_01, swing1LimitDegrees=5, swing2LimitDegrees=30, twistLimitDegrees=5}
 {bone1Name=dyn_hair_bk_03, bone2Name=dyn_hair_bk_02, swing1LimitDegrees=5, swing2LimitDegrees=30, twistLimitDegrees=5}
 {bone1Name=dyn_hair_bk_04, bone2Name=dyn_hair_bk_03, swing1LimitDegrees=5, swing2LimitDegrees=30, twistLimitDegrees=5}
 {bone1Name=dyn_hair_bk_01_l, bone2Name=head, swing1LimitDegrees=5, swing2LimitDegrees=20, twistLimitDegrees=5}
 {bone1Name=dyn_hair_bk_02_l, bone2Name=dyn_hair_bk_01_l, swing1LimitDegrees=5, swing2LimitDegrees=25, twistLimitDegrees=5}
 {bone1Name=dyn_hair_bk_03_l, bone2Name=dyn_hair_bk_02_l, swing1LimitDegrees=10, swing2LimitDegrees=30, twistLimitDegrees=5}
 {bone1Name=dyn_hair_bk_04_l, bone2Name=dyn_hair_bk_03_l, swing1LimitDegrees=10, swing2LimitDegrees=20, twistLimitDegrees=5}
 {bone1Name=dyn_hair_01_l, bone2Name=head, swing1LimitDegrees=10, swing2LimitDegrees=30, twistLimitDegrees=5}
 {bone1Name=dyn_hair_02_l, bone2Name=dyn_hair_01_l, swing1LimitDegrees=15, swing2LimitDegrees=30, twistLimitDegrees=5}
 {bone1Name=dyn_hair_03_l, bone2Name=dyn_hair_02_l, swing1LimitDegrees=20, swing2LimitDegrees=30, twistLimitDegrees=5}
 {bone1Name=dyn_hair_04_l, bone2Name=dyn_hair_03_l, swing1LimitDegrees=10, swing2LimitDegrees=30, twistLimitDegrees=5}
 {bone1Name=dyn_hair_fr_01_l, bone2Name=head, swing1LimitDegrees=10, swing2LimitDegrees=10, twistLimitDegrees=5}
 {bone1Name=dyn_hair_fr_02_l, bone2Name=dyn_hair_fr_01_l, swing1LimitDegrees=10, swing2LimitDegrees=30, twistLimitDegrees=5}
 {bone1Name=dyn_hair_fr_03_l, bone2Name=dyn_hair_fr_02_l, swing1LimitDegrees=15, swing2LimitDegrees=35, twistLimitDegrees=5}
 {bone1Name=dyn_hair_fr_04_l, bone2Name=dyn_hair_fr_03_l, swing1LimitDegrees=30, swing2LimitDegrees=35, twistLimitDegrees=5}
 {bone1Name=dyn_hair_bk_01_r, bone2Name=head, swing1LimitDegrees=15, swing2LimitDegrees=20, twistLimitDegrees=5}
 {bone1Name=dyn_hair_bk_02_r, bone2Name=dyn_hair_bk_01_r, swing1LimitDegrees=5, swing2LimitDegrees=30, twistLimitDegrees=5}
 {bone1Name=dyn_hair_bk_03_r, bone2Name=dyn_hair_bk_02_r, swing1LimitDegrees=5, swing2LimitDegrees=27, twistLimitDegrees=5}
 {bone1Name=dyn_hair_01_r, bone2Name=head, swing1LimitDegrees=20, swing2LimitDegrees=30, twistLimitDegrees=5}
 {bone1Name=dyn_hair_02_r, bone2Name=dyn_hair_01_r, swing1LimitDegrees=20, swing2LimitDegrees=30, twistLimitDegrees=10}
 {bone1Name=dyn_hair_03_r, bone2Name=dyn_hair_02_r, swing1LimitDegrees=20, swing2LimitDegrees=20, twistLimitDegrees=10}
 {bone1Name=dyn_hair_04_r, bone2Name=dyn_hair_03_r, swing1LimitDegrees=30, swing2LimitDegrees=40, twistLimitDegrees=5}
 {bone1Name=dyn_hair_fr_01_r, bone2Name=head, swing1LimitDegrees=20, swing2LimitDegrees=10, twistLimitDegrees=5}
 {bone1Name=dyn_hair_fr_02_r, bone2Name=dyn_hair_fr_01_r, swing1LimitDegrees=25, swing2LimitDegrees=15, twistLimitDegrees=5}
 {bone1Name=dyn_hair_fr_03_r, bone2Name=dyn_hair_fr_02_r, swing1LimitDegrees=25, swing2LimitDegrees=15, twistLimitDegrees=5}
 {bone1Name=dyn_hair_fr_04_r, bone2Name=dyn_hair_fr_03_r, swing1LimitDegrees=45, swing2LimitDegrees=45, twistLimitDegrees=5}
 {
  bone1Name=head
  bone2Name=neck_01
  swing1LimitDegrees=45
  swing1Motion=Free
  swing2LimitDegrees=45
  swing2Motion=Free
  twistLimitDegrees=45
  twistMotion=Free
 }
 {
  bone1Name=clavicle_l
  bone2Name=spine_05
  swing1LimitDegrees=45
  swing1Motion=Free
  swing2LimitDegrees=45
  swing2Motion=Free
  twistLimitDegrees=45
  twistMotion=Free
 }
 {
  bone1Name=clavicle_r
  bone2Name=spine_05
  swing1LimitDegrees=45
  swing1Motion=Free
  swing2LimitDegrees=45
  swing2Motion=Free
  twistLimitDegrees=45
  twistMotion=Free
 }
 {
  bone1Name=neck_01
  bone2Name=spine_05
  swing1LimitDegrees=45
  swing1Motion=Free
  swing2LimitDegrees=45
  swing2Motion=Free
  twistLimitDegrees=45
  twistMotion=Free
 }
 {bone1Name=dyn_hair_bk_04_r, bone2Name=dyn_hair_bk_03_r, swing1LimitDegrees=5, swing2LimitDegrees=20, twistLimitDegrees=5}
 {
  bone1Name=upperarm_l
  bone2Name=clavicle_l
  swing1LimitDegrees=45
  swing1Motion=Free
  swing2LimitDegrees=45
  swing2Motion=Free
  twistLimitDegrees=45
  twistMotion=Free
 }
 {
  bone1Name=upperarm_r
  bone2Name=clavicle_r
  swing1LimitDegrees=45
  swing1Motion=Free
  swing2LimitDegrees=45
  swing2Motion=Free
  twistLimitDegrees=45
  twistMotion=Free
 }
]
